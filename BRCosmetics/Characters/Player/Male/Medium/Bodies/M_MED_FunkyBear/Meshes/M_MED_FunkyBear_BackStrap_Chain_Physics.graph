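# Physics asset for the FunkyBear medium male body that defines capsule bodies and joints for its left-side back strap and front chain.
# Strap and chain are each five-link series hung from a base bone, with limited swing and zero-degree twist limits.
# Pelvis and left thigh capsules supply body volumes.
# Two jacket end capsules are included, but only the first is constrained, to the pelvis with single-axis swing.
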PhysicsAsset /BRCosmetics/Characters/Player/Male/Medium/Bodies/M_MED_FunkyBear/Meshes/M_MED_FunkyBear_BackStrap_Chain_Physics
bodies {
 base_chain_front_01_lf [{shapeType=Capsule, length=0.505, radius=0.505, shapeName=base_chain_front_01_lf_capsule, rotation=<rot 0 -90 0>}]
 base_strap_back_01_lf [{shapeType=Capsule, length=0.505, radius=0.505, shapeName=base_strap_back_01_lf_capsule, rotation=<rot 0 -90 0>}]
 dyn_chain_front_01_lf [{
   shapeType=Capsule
   length=2.4279
   radius=0.5221
   shapeName=dyn_chain_front_01_lf_capsule
   center=<3 0 0>
   rotation=<rot 80.0273 6.8192 -0.0002>
  }]
 dyn_chain_front_02_lf [{
   shapeType=Capsule
   length=2.2817
   radius=0.6683
   shapeName=dyn_chain_front_02_lf_capsule
   center=<3 0 0>
   rotation=<rot 85.1117 -0.0003 -0.0004>
  }]
 dyn_chain_front_03_lf [{shapeType=Capsule, length=2.1854, radius=0.7646, shapeName=dyn_chain_front_03_lf_capsule, center=<3 0 0>, rotation=<rot 90 0 0>}]
 dyn_chain_front_04_lf [{shapeType=Capsule, length=2.1923, radius=0.7577, shapeName=dyn_chain_front_04_lf_capsule, center=<3 0 0>, rotation=<rot 90 0 0>}]
 dyn_chain_front_05_lf [{shapeType=Capsule, length=2.4637, radius=3.9697, shapeName=dyn_chain_front_05_lf_capsule, center=<4.8 -0.65 -1.05>, rotation=<rot 0 0 -66.9>}]
 dyn_jacket_01_lf_end [{shapeType=Capsule, length=6, radius=3, shapeName=dyn_jacket_01_lf_end_capsule, center=<5 0 0>, rotation=<rot -3.2484 0 0>}]
 dyn_jacket_02_lf_end [{shapeType=Capsule, length=6, radius=3, shapeName=dyn_jacket_02_lf_end_capsule, center=<5 0 0>}]
 dyn_strap_back_01_lf [{shapeType=Capsule, length=5.1128, radius=1.9592, shapeName=dyn_strap_back_01_lf_capsule, center=<2 0 0>, rotation=<rot 95 0 0>}]
 dyn_strap_back_02_lf [{
   shapeType=Capsule
   length=5.6131
   radius=1.7858
   shapeName=dyn_strap_back_02_lf_capsule
   center=<2 0 0.5>
   rotation=<rot 80.2567 -179.9999 -180>
  }]
 dyn_strap_back_03_lf [{shapeType=Capsule, length=5.4772, radius=1.9274, shapeName=dyn_strap_back_03_lf_capsule, center=<2 0 0.8>, rotation=<rot 90 0 0>}]
 dyn_strap_back_04_lf [{shapeType=Capsule, length=4.9114, radius=1.7476, shapeName=dyn_strap_back_04_lf_capsule, center=<2 0 0>, rotation=<rot 90 0 0>}]
 dyn_strap_back_05_lf [
  {shapeType=Capsule, length=5.2616, radius=1.7476, shapeName=dyn_strap_back_05_lf_capsule, center=<3 0 0>, rotation=<rot 90 0 0>}
  {shapeType=Capsule, length=5.8121, radius=2.038, shapeName=dyn_strap_back_05_lf_capsule, center=<2 0 0>, rotation=<rot -0.0108 0.0032 84.005>}
 ]
 pelvis [{shapeType=Capsule, length=4.4698, radius=11.5798, shapeName=pelvis_capsule, center=<-1.1436 -0.9852 0>}]
 thigh_l [{shapeType=Capsule, length=48.2264, radius=8.483, shapeName=thigh_l_capsule, center=<-24.3385 -0.0947 0>, rotation=<rot 90 -4.2253 0>}]
}
constraints [
 {bone1Name=dyn_strap_back_01_lf, bone2Name=base_strap_back_01_lf, swing1LimitDegrees=25, swing2LimitDegrees=20, twistLimitDegrees=0}
 {bone1Name=dyn_strap_back_02_lf, bone2Name=dyn_strap_back_01_lf, swing1LimitDegrees=25, swing2LimitDegrees=20, twistLimitDegrees=0}
 {bone1Name=dyn_strap_back_03_lf, bone2Name=dyn_strap_back_02_lf, swing1LimitDegrees=20, swing2LimitDegrees=20, twistLimitDegrees=0}
 {bone1Name=dyn_strap_back_04_lf, bone2Name=dyn_strap_back_03_lf, swing1LimitDegrees=30, swing2LimitDegrees=20, twistLimitDegrees=0}
 {bone1Name=dyn_strap_back_05_lf, bone2Name=dyn_strap_back_04_lf, swing1LimitDegrees=40, swing2LimitDegrees=20, twistLimitDegrees=0}
 {bone1Name=dyn_chain_front_01_lf, bone2Name=base_chain_front_01_lf, swing1LimitDegrees=5, swing2LimitDegrees=5, twistLimitDegrees=0}
 {bone1Name=dyn_chain_front_02_lf, bone2Name=dyn_chain_front_01_lf, swing1LimitDegrees=20, swing2LimitDegrees=20, twistLimitDegrees=0}
 {bone1Name=dyn_chain_front_03_lf, bone2Name=dyn_chain_front_02_lf, swing1LimitDegrees=20, swing2LimitDegrees=20, twistLimitDegrees=0}
 {bone1Name=dyn_chain_front_04_lf, bone2Name=dyn_chain_front_03_lf, swing1LimitDegrees=20, swing2LimitDegrees=20, twistLimitDegrees=0}
 {bone1Name=dyn_chain_front_05_lf, bone2Name=dyn_chain_front_04_lf, swing1LimitDegrees=10, swing2LimitDegrees=10, twistLimitDegrees=0}
 {bone1Name=dyn_jacket_01_lf_end, bone2Name=pelvis, swing1LimitDegrees=10, swing2LimitDegrees=0, twistLimitDegrees=0}
]
