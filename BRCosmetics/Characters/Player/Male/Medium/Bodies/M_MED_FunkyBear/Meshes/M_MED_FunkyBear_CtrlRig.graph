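# Control Rig for the M_MED_FunkyBear player cosmetic body, procedurally posing accessory bones such as gun, strap, chain, necklace, fur and jacket.
# Weighted parent constraints attach accessories to the body, while swing extraction, reversed spine rotations and an aim constraint adjust them.
# Spherical pose readers and delta-rotation weights blend corrective offsets, and spherical collisions act between fur bones and head.
# Smoothed pelvis-height crouch alpha drives chain and fur fixes and a "crouching alpha" curve.
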ControlRigBlueprint /BRCosmetics/Characters/Player/Male/Medium/Bodies/M_MED_FunkyBear/Meshes/M_MED_FunkyBear_CtrlRig
graphs [
 New Function_ContainedGraph {nodes [
   RigVMFunctionEntryNode()
   > &RigUnit_ParentConstraint(
    AdvancedSettings=(RotationOrderForFilter=XZY)
    Child=(Type=Bone,Name="base_gun_rt")
    Parents="((Item=(Type=Bone,Name=\"deform_glute_r\"),Weight=0.0045),(Item=(Type=Bone,Name=\"deform_groin_r\"),Weight=0.0153),(Item=(Type=Bone,Name=\"spine_03\"),Weight=0.0277),(Item=(Type=Bone,Name=\"spine_01\"),Weight=0.0434),(Item=(Type=Bone,Name=\"thigh_twist_01_r\"),Weigh ...(truncated)"
    Weight=1
   )
   > &RigUnit_ParentConstraint(
    AdvancedSettings=(InterpolationType=Average,RotationOrderForFilter=XZY)
    Child=(Type=Bone,Name="thigh_r_get_rot")
    Parents=((Item=(Type=Bone,Name="thigh_r"),Weight=1))
    Weight=1
   )
   > get_swing_axis_front = RigVMCollapseNode(
    Axis to Get=(X=0,Y=1,Z=0)
    Backward=false
    Forward=false
    Inward=false
    Outward=true
    Rotation Order=ZXY
    Twist Axis=(X=1,Y=0,Z=0)
    Twist_Inward=false
    Twist_Outward=false
    input joint=thigh_r_get_rot
   )
   RigVMFunction_MathQuaternionToEuler = &RigVMFunction_MathQuaternionToEuler(RotationOrder=ZYX, Value=@get_swing_axis_front.Output Rotate.Value)
   RigVMFunction_MathQuaternionFromEuler = &RigVMFunction_MathQuaternionFromEuler(Euler=/Script/RigVM.RigVMFunction_MathVectorMul(B=(X=1,Y=1,Z=1), A=@RigVMFunction_MathQuaternionToEuler.Result) {is=struct RigVMUnitNode, template="Multiply::Execute(in A,in B,out Result)"}.Result)
   > &RigUnit_ModifyTransforms(
    ItemToModify=((Item=(Type=Bone,Name="thigh_r_set_rot"),Transform=(Rotation=(X=0,Y=0,Z=0,W=1),Translation=(X=0,Y=0,Z=0),Scale3D=(X=1,Y=1,Z=1))))
    Weight=1
    WeightMinimum=0
    ItemToModify.0.Transform.Rotation=@RigVMFunction_MathQuaternionFromEuler.Result
   )
   > &RigUnit_ModifyTransforms(
    ItemToModify=((Item=(Type=Bone,Name="thigh_r_set_rot"),Transform=(Rotation=(X=0,Y=0.0305,Z=0,W=0.9995),Translation=(X=0,Y=0,Z=0),Scale3D=(X=1,Y=1,Z=1))))
    Weight=1
    WeightMinimum=0
   )
   > &RigUnit_RotationConstraintLocalSpaceOffset(
    AdvancedSettings=(InterpolationType=Average,RotationOrderForFilter=XZY)
    Child=(Type=Bone,Name="base_gun_rt")
    Filter=(bX=True,bY=True,bZ=True)
    Parents=((Item=(Type=Bone,Name="thigh_r_set_rot"),Weight=1))
    Weight=1
   )
   > n1 = &RigUnit_SphericalPoseReader(
    ActiveRegionScaleFactors=(PositiveWidth=1,NegativeWidth=1,PositiveHeight=1,NegativeHeight=1)
    ActiveRegionSize=0.1
    Debug=(bDrawDebug=false,bDraw2D=False,bDrawLocalAxes=False,DebugScale=25,DebugSegments=20,DebugThickness=0.25)
    DriverAxis=(X=1,Y=0,Z=0)
    DriverItem=(Type=Bone,Name="thigh_r")
    FalloffRegionScaleFactors=(PositiveWidth=1.25,NegativeWidth=0.9,PositiveHeight=0.9,NegativeHeight=0.9)
    FalloffSize=0.5
    FlipHeightScaling=False
    FlipWidthScaling=False
    OptionalParentItem=(Type=Bone,Name="None")
    RotationOffset=(X=-90,Y=0,Z=-20)
   )
   AlphaInterp = &RigVMFunction_AlphaInterp(InterpSpeedDecreasing=10, InterpSpeedIncreasing=10, Scale=1, bClampResult=False, Value=@n1.OutputParam)
   > &RigUnit_ModifyTransforms(
    ItemToModify=((Item=(Type=Bone,Name="base_gun_rt"),Transform=(Rotation=(X=-0.1392,Y=0,Z=0,W=0.9903),Translation=(X=0,Y=0,Z=0),Scale3D=(X=1,Y=1,Z=1))))
    WeightMinimum=0
    Weight=@AlphaInterp.Result
   )
   > RigVMFunctionReturnNode()
   Remap = &RigVMFunction_MathDoubleRemap(SourceMaximum=1, SourceMinimum=0, TargetMaximum=0.75, TargetMinimum=0.65, Value=RigVMVariableNode() {variable=isCrouchingMoveCR}.Value)
   &RigVMFunction_AlphaInterp(InterpSpeedDecreasing=10, InterpSpeedIncreasing=10, Scale=1, bClampResult=true, bMapRange=true, Value=@Remap.Result)
  ]}
 New Function_ContainedGraph {nodes [
   n1 = RigVMFunctionEntryNode()
   RigUnit_GetTransform = &RigUnit_GetTransform(Item=(Type=Bone,Name="thigh_l_for_cons"), Space=LocalSpace, bInitial=false, Item.Name=@n1.input joint)
   RigVMFunction_MathQuaternionToEuler = &RigVMFunction_MathQuaternionToEuler(
    RotationOrder=@n1.Rotation Order
    Value=&RigVMFunction_MathQuaternionSwingTwist(
     TwistAxis=(X=1,Y=0,Z=0)
     TwistAxis.X=@n1.Twist Axis.X
     TwistAxis.Y=@n1.Twist Axis.Y
     TwistAxis.Z=@n1.Twist Axis.Z
     Input=@RigUnit_GetTransform.Transform.Rotation
    ).Swing
   )
   n2 = &RigVMFunction_MathDoubleMul(B=@RigVMFunction_MathQuaternionToEuler.Result.Z, A=@n1.Axis to Get.Z)
   n3 = /Script/RigVM.RigVMFunction_MathDoubleLess(B=0, A=@n2.Result) {is=struct RigVMUnitNode, template="Less::Execute(in A,in B,out Result)"}
   n4 = /Script/RigVM.RigVMFunction_MathDoubleGreater(B=0, A=@n2.Result) {is=struct RigVMUnitNode, template="Greater::Execute(in A,in B,out Result)"}
   Add = &RigVMFunction_MathDoubleAdd(
    A=RigVMDispatchNode(
     False=0
     Condition=@n1.Forward
     True=RigVMDispatchNode(False=0, Condition=@n3.Result, True=@RigVMFunction_MathQuaternionToEuler.Result.Z) {template="DISPATCH_RigVMDispatch_If(in Condition,in True,in False,out Result)"}.Result
    ) {template="DISPATCH_RigVMDispatch_If(in Condition,in True,in False,out Result)"}.Result
    B=RigVMDispatchNode(
     False=0
     Condition=@n1.Backward
     True=RigVMDispatchNode(False=0, Condition=@n4.Result, True=@RigVMFunction_MathQuaternionToEuler.Result.Z) {template="DISPATCH_RigVMDispatch_If(in Condition,in True,in False,out Result)"}.Result
    ) {template="DISPATCH_RigVMDispatch_If(in Condition,in True,in False,out Result)"}.Result
   )
   n5 = &RigVMFunction_MathDoubleMul(A=@n1.Axis to Get.Y, B=@RigVMFunction_MathQuaternionToEuler.Result.Y)
   Less = &RigVMFunction_MathDoubleLess(B=0, A=@n5.Result) {template="Less::Execute(in A,in B,out Result)"}
   n6 = RigVMDispatchNode(False=0, Condition=@Less.Result, True=@RigVMFunction_MathQuaternionToEuler.Result.Y) {template="DISPATCH_RigVMDispatch_If(in Condition,in True,in False,out Result)"}
   Greater = &RigVMFunction_MathDoubleGreater(B=0, A=@n5.Result) {template="Greater::Execute(in A,in B,out Result)"}
   n7 = RigVMDispatchNode(False=0, Condition=@Greater.Result, True=@RigVMFunction_MathQuaternionToEuler.Result.Y) {template="DISPATCH_RigVMDispatch_If(in Condition,in True,in False,out Result)"}
   RigVMFunction_MathQuaternionFromEuler = &RigVMFunction_MathQuaternionFromEuler(
    Euler=(X=0,Y=0,Z=0)
    Euler.Z=@Add.Result
    Euler.Y=&RigVMFunction_MathDoubleAdd(
     A=RigVMDispatchNode(False=0, Condition=@n1.Outward, True=@n6.Result) {template="DISPATCH_RigVMDispatch_If(in Condition,in True,in False,out Result)"}.Result
     B=RigVMDispatchNode(False=0, Condition=@n1.Inward, True=@n7.Result) {template="DISPATCH_RigVMDispatch_If(in Condition,in True,in False,out Result)"}.Result
    ).Result
    RotationOrder=@n1.Rotation Order
   )
   > RigVMFunctionReturnNode(Output Rotate=(Value=(X=0,Y=0,Z=0,W=1)), Output Rotate.Value=@RigVMFunction_MathQuaternionFromEuler.Result)
  ]}
 New Function_ContainedGraph {nodes [
   n1 = RigVMFunctionEntryNode()
   RigVMFunction_MathBoolToInteger = &RigVMFunction_MathBoolToInteger(Value=@n1.onlyTwist)
   RigVMFunction_MathQuaternionSwingTwist = &RigVMFunction_MathQuaternionSwingTwist(
    twistAxis=@n1.twistAxis
    Input=&RigUnit_GetRelativeTransformForItem(
     bChildInitial=False
     bParentInitial=False
     child=(Type=Bone,Name="None")
     parent=(Type=Bone,Name="None")
     child.Name=@n1.child
     parent.Name=@n1.parent
     parent.Type=@n1.parentType
    ).RelativeTransform.Rotation
   )
   RigVMFunction_MathQuaternionToEuler = &RigVMFunction_MathQuaternionToEuler(
    Value=RigVMDispatchNode(
     Values=((),())
     Index=@RigVMFunction_MathBoolToInteger.Result
     Values.0=@RigVMFunction_MathQuaternionSwingTwist.Swing
     Values.1=@RigVMFunction_MathQuaternionSwingTwist.Twist
    ) {template="DISPATCH_RigVMDispatch_SelectInt32(in Index,in Values,out Result)"}.Result
    RotationOrder=@n1.Rotation Order
   )
   > DISPATCH_RigVMDispatch_Print = RigVMDispatchNode(ScreenColor=(R=1,G=1,B=1,A=1), ScreenDuration=1.5, Enabled=@n1.debugRotation, Value=@RigVMFunction_MathQuaternionToEuler.Result) {template="DISPATCH_RigVMDispatch_Print(in Prefix,in Value,in Enabled,in ScreenDuration,in ScreenColor)"}
   Clamp = &RigVMFunction_MathIntClamp(Value=@n1.primAxis)
   DISPATCH_RigVMDispatch_SelectInt32 = RigVMDispatchNode(
    Values=(0,0,0)
    Index=@Clamp.Result
    Values.0=@RigVMFunction_MathQuaternionToEuler.Result.X
    Values.1=@RigVMFunction_MathQuaternionToEuler.Result.Y
    Values.2=@RigVMFunction_MathQuaternionToEuler.Result.Z
   ) {template="DISPATCH_RigVMDispatch_SelectInt32(in Index,in Values,out Result)"}
   Remap = &RigVMFunction_MathDoubleRemap(
    TargetMaximum=1
    TargetMinimum=0
    bClamp=False
    Value=@DISPATCH_RigVMDispatch_SelectInt32.Result
    SourceMaximum=@n1.max
    SourceMinimum=@n1.min
   )
   Multiply = &RigVMFunction_MathDoubleMul(A=&RigVMFunction_MathDoubleClamp(Value=@Remap.Result).Result, B=@n1.premult)
   > RigVMFunctionReturnNode(outputWeight=@Multiply.Result)
  ]}
 New Function_ContainedGraph {nodes [
   RigVMFunctionEntryNode()
   > &RigUnit_ParentConstraint(
    AdvancedSettings=(RotationOrderForFilter=XZY)
    Child=(Type=Bone,Name="base_strap_back_01_lf")
    Parents=((Item=(Type=Bone,Name="spine_03"),Weight=0.0568),(Item=(Type=Bone,Name="spine_01"),Weight=0.082),(Item=(Type=Bone,Name="spine_02"),Weight=0.0879),(Item=(Type=Bone,Name="pelvis"),Weight=0.7733))
    Weight=1
   )
   > n1 = &RigUnit_SphericalPoseReader(
    ActiveRegionScaleFactors=(PositiveWidth=1,NegativeWidth=1,PositiveHeight=1,NegativeHeight=1)
    ActiveRegionSize=0.1
    Debug=(bDrawDebug=false,bDraw2D=False,bDrawLocalAxes=False,DebugScale=25,DebugSegments=20,DebugThickness=0.25)
    DriverAxis=(X=-1,Y=0,Z=0)
    DriverItem=(Type=Bone,Name="thigh_l")
    FalloffRegionScaleFactors=(PositiveWidth=1,NegativeWidth=1,PositiveHeight=1,NegativeHeight=0.8)
    FalloffSize=0.555
    FlipHeightScaling=False
    FlipWidthScaling=False
    OptionalParentItem=(Type=Bone,Name="None")
    RotationOffset=(X=-90,Y=0,Z=20)
   )
   > &RigUnit_ModifyTransforms(
    ItemToModify=((Item=(Type=Bone,Name="base_strap_back_01_lf"),Transform=(Rotation=(X=0,Y=-0.5592,Z=0,W=0.829),Translation=(X=0,Y=0,Z=0),Scale3D=(X=1,Y=1,Z=1))))
    WeightMinimum=0
    Weight=@n1.OutputParam
   )
   RigVMFunction_MathQuaternionToEuler = &RigVMFunction_MathQuaternionToEuler(
    RotationOrder=ZYX
    Value=&RigUnit_GetRelativeTransformForItem(Child=(Type=Bone,Name="pelvis"), Parent=(Type=Bone,Name="spine_03"), bChildInitial=False, bParentInitial=False).RelativeTransform.Rotation
   )
   Remap = &RigVMFunction_MathDoubleRemap(SourceMaximum=69.4, SourceMinimum=8.33, TargetMaximum=1, TargetMinimum=0, Value=@RigVMFunction_MathQuaternionToEuler.Result.Z)
   > &RigUnit_ModifyTransforms(
    ItemToModify=((Item=(Type=Bone,Name="base_strap_back_01_lf"),Transform=(Rotation=(X=0,Y=-0.1045,Z=0,W=0.9945),Translation=(X=0,Y=0,Z=0),Scale3D=(X=1,Y=1,Z=1))))
    WeightMinimum=0
    Weight=@Remap.Result
   )
   > RigVMFunctionReturnNode()
  ]}
 New Function_ContainedGraph {nodes [
   RigVMFunctionEntryNode()
   > &RigUnit_ParentConstraint(
    AdvancedSettings=(RotationOrderForFilter=XZY)
    Child=(Type=Bone,Name="base_chain_front_01_lf")
    Parents="((Item=(Type=Bone,Name=\"spine_03\"),Weight=0.0023),(Item=(Type=Bone,Name=\"spine_01\"),Weight=0.0041),(Item=(Type=Bone,Name=\"spine_02\"),Weight=0.0053),(Item=(Type=Bone,Name=\"thigh_twist_01_l\"),Weight=0.1958),(Item=(Type=Bone,Name=\"deform_groin_l\"),Weight=0.22 ...(truncated)"
    Weight=1
   )
   > &RigUnit_ParentConstraint(
    AdvancedSettings=(InterpolationType=Average,RotationOrderForFilter=XZY)
    Child=(Type=Bone,Name="thigh_l_get_rot")
    Parents=((Item=(Type=Bone,Name="thigh_l"),Weight=1))
    Weight=1
   )
   > get_swing_axis_front = RigVMCollapseNode(
    Axis to Get=(X=0,Y=1,Z=1)
    Backward=false
    Forward=true
    Inward=false
    Outward=true
    Rotation Order=ZXY
    Twist Axis=(X=1,Y=0,Z=0)
    Twist_Inward=false
    Twist_Outward=false
    input joint=thigh_l_get_rot
   )
   > &RigUnit_ModifyTransforms(
    ItemToModify=((Item=(Type=Bone,Name="thigh_l_set_rot"),Transform=(Rotation=(X=0,Y=0,Z=0,W=1),Translation=(X=0,Y=0,Z=0),Scale3D=(X=1,Y=1,Z=1))))
    Weight=1
    WeightMinimum=0
    ItemToModify.0.Transform.Rotation=@get_swing_axis_front.Output Rotate.Value
   )
   > &RigUnit_RotationConstraintLocalSpaceOffset(
    AdvancedSettings=(InterpolationType=Average,RotationOrderForFilter=XZY)
    Child=(Type=Bone,Name="base_chain_front_01_lf")
    Filter=(bX=True,bY=True,bZ=True)
    Parents=((Item=(Type=Bone,Name="thigh_l_set_rot"),Weight=1))
    Weight=1
   )
   > &RigUnit_ModifyTransforms(
    ItemToModify=((Item=(Type=Bone,Name="base_chain_front_01_lf"),Transform=(Rotation=(X=0,Y=0.061,Z=0,W=0.9981),Translation=(X=0,Y=0,Z=0),Scale3D=(X=1,Y=1,Z=1))))
    Weight=1
    WeightMinimum=0
   )
   Remap = &RigVMFunction_MathDoubleRemap(
    SourceMaximum=45
    SourceMinimum=89
    TargetMaximum=1
    TargetMinimum=0
    Value=&RigUnit_GetRelativeTransformForItem(Child=(Type=Bone,Name="pelvis"), Parent=(Type=Bone,Name="root"), bChildInitial=False, bParentInitial=False).RelativeTransform.Translation.Z
   )
   RigVMFunction_MathBoolToFloat = &RigVMFunction_MathBoolToFloat(Value=RigVMVariableNode() {variable=isCrouching}.Value)
   Multiply = &RigVMFunction_MathDoubleMul(A=@Remap.Result, B=@RigVMFunction_MathBoolToFloat.Result)
   AlphaInterp = &RigVMFunction_AlphaInterp(InterpSpeedDecreasing=10, InterpSpeedIncreasing=10, Scale=1, bClampResult=False, Value=@Multiply.Result)
   > &RigUnit_ModifyTransforms(
    ItemToModify=((Item=(Type=Bone,Name="base_chain_front_01_lf"),Transform=(Rotation=(X=-0.2271,Y=0.1624,Z=0.731,W=-0.6227),Translation=(X=0,Y=0,Z=0),Scale3D=(X=1,Y=1,Z=1))))
    WeightMinimum=0
    Weight=@AlphaInterp.Result
   )
   > RigVMVariableNode(Value=@AlphaInterp.Result) {variable=isCrouchingCR}
   Subtract = &RigVMFunction_MathDoubleSub(B=1, A=@AlphaInterp.Result)
   Absolute = &RigVMFunction_MathDoubleAbs(Value=@Subtract.Result)
   > /Script/ControlRig.RigUnit_SetCurveValue(Curve=crouching alpha, Value=@Absolute.Result) {is=struct RigVMUnitNode}
   > RigVMFunctionReturnNode()
  ]}
 New Function_ContainedGraph {nodes [
   n1 = RigVMFunctionEntryNode()
   RigUnit_GetTransform = &RigUnit_GetTransform(Item=(Type=Bone,Name="thigh_l_for_cons"), Space=LocalSpace, bInitial=false, Item.Name=@n1.input joint)
   RigVMFunction_MathQuaternionToEuler = &RigVMFunction_MathQuaternionToEuler(
    RotationOrder=@n1.Rotation Order
    Value=&RigVMFunction_MathQuaternionSwingTwist(
     TwistAxis=(X=1,Y=0,Z=0)
     TwistAxis.X=@n1.Twist Axis.X
     TwistAxis.Y=@n1.Twist Axis.Y
     TwistAxis.Z=@n1.Twist Axis.Z
     Input=@RigUnit_GetTransform.Transform.Rotation
    ).Swing
   )
   n2 = &RigVMFunction_MathDoubleMul(B=@RigVMFunction_MathQuaternionToEuler.Result.Z, A=@n1.Axis to Get.Z)
   n3 = &RigVMFunction_MathDoubleLess(B=0, A=@n2.Result) {template="Less::Execute(in A,in B,out Result)"}
   n4 = &RigVMFunction_MathDoubleGreater(B=0, A=@n2.Result) {template="Greater::Execute(in A,in B,out Result)"}
   Add = &RigVMFunction_MathDoubleAdd(
    A=RigVMDispatchNode(
     False=0
     Condition=@n1.Forward
     True=RigVMDispatchNode(False=0, Condition=@n3.Result, True=@RigVMFunction_MathQuaternionToEuler.Result.Z) {template="DISPATCH_RigVMDispatch_If(in Condition,in True,in False,out Result)"}.Result
    ) {template="DISPATCH_RigVMDispatch_If(in Condition,in True,in False,out Result)"}.Result
    B=RigVMDispatchNode(
     False=0
     Condition=@n1.Backward
     True=RigVMDispatchNode(False=0, Condition=@n4.Result, True=@RigVMFunction_MathQuaternionToEuler.Result.Z) {template="DISPATCH_RigVMDispatch_If(in Condition,in True,in False,out Result)"}.Result
    ) {template="DISPATCH_RigVMDispatch_If(in Condition,in True,in False,out Result)"}.Result
   )
   n5 = &RigVMFunction_MathDoubleMul(A=@n1.Axis to Get.Y, B=@RigVMFunction_MathQuaternionToEuler.Result.Y)
   Less = &RigVMFunction_MathDoubleLess(B=0, A=@n5.Result) {template="Less::Execute(in A,in B,out Result)"}
   n6 = RigVMDispatchNode(False=0, Condition=@Less.Result, True=@RigVMFunction_MathQuaternionToEuler.Result.Y) {template="DISPATCH_RigVMDispatch_If(in Condition,in True,in False,out Result)"}
   Greater = &RigVMFunction_MathDoubleGreater(B=0, A=@n5.Result) {template="Greater::Execute(in A,in B,out Result)"}
   n7 = RigVMDispatchNode(False=0, Condition=@Greater.Result, True=@RigVMFunction_MathQuaternionToEuler.Result.Y) {template="DISPATCH_RigVMDispatch_If(in Condition,in True,in False,out Result)"}
   RigVMFunction_MathQuaternionFromEuler = &RigVMFunction_MathQuaternionFromEuler(
    Euler=(X=0,Y=0,Z=0)
    Euler.Z=@Add.Result
    Euler.Y=&RigVMFunction_MathDoubleAdd(
     A=RigVMDispatchNode(False=0, Condition=@n1.Outward, True=@n6.Result) {template="DISPATCH_RigVMDispatch_If(in Condition,in True,in False,out Result)"}.Result
     B=RigVMDispatchNode(False=0, Condition=@n1.Inward, True=@n7.Result) {template="DISPATCH_RigVMDispatch_If(in Condition,in True,in False,out Result)"}.Result
    ).Result
    RotationOrder=@n1.Rotation Order
   )
   > RigVMFunctionReturnNode(Output Rotate=(Value=(X=0,Y=0,Z=0,W=1)), Output Rotate.Value=@RigVMFunction_MathQuaternionFromEuler.Result)
  ]}
 New Function_ContainedGraph {nodes [
   RigVMFunctionEntryNode()
   > &RigUnit_ParentConstraint(
    AdvancedSettings=(RotationOrderForFilter=XZY)
    Child=(Type=Bone,Name="base_necklace")
    Parents="((Item=(Type=Bone,Name=\"clavicle_r\"),Weight=0.0013),(Item=(Type=Bone,Name=\"spine_04\"),Weight=0.0036),(Item=(Type=Bone,Name=\"head\"),Weight=0.0415),(Item=(Type=Bone,Name=\"neck_01\"),Weight=0.0715),(Item=(Type=Bone,Name=\"neck_02\"),Weight=0.1692),(Item=(Type=Bo ...(truncated)"
    Weight=1
   )
   > &RigUnit_ParentConstraint(
    AdvancedSettings=(RotationOrderForFilter=XZY)
    Child=(Type=Bone,Name="aim_necklace")
    Parents="((Item=(Type=Bone,Name=\"clavicle_l\"),Weight=0.049),(Item=(Type=Bone,Name=\"clavicle_r\"),Weight=0.0492),(Item=(Type=Bone,Name=\"deform_pec_l\"),Weight=0.0637),(Item=(Type=Bone,Name=\"deform_pec_r\"),Weight=0.0638),(Item=(Type=Bone,Name=\"spine_03\"),Weight=0.081), ...(truncated)"
    Weight=1
   )
   > &RigUnit_AimConstraintLocalSpaceOffset(
    Child=(Type=Bone,Name="base_necklace")
    Filter=(bX=True,bY=True,bZ=True)
    Parents=((Item=(Type=Bone,Name="aim_necklace"),Weight=1))
    Weight=1
    WorldUp=(Target=(X=0,Y=0,Z=1),Kind=Direction,Space=(Type=Bone,Name="aim_necklace"))
   )
   > RigVMFunctionReturnNode()
   &RigUnit_ParentConstraint(
    AdvancedSettings=(RotationOrderForFilter=XZY)
    Child=(Type=Bone,Name="aim_necklace")
    ExecutePin=()
    Filter=(TranslationFilter=(bX=True,bY=True,bZ=True),RotationFilter=(bX=true,bY=true,bZ=true),ScaleFilter=(bX=True,bY=True,bZ=True))
    Parents="((Item=(Type=Bone,Name=\"clavicle_l\"),Weight=0.0007),(Item=(Type=Bone,Name=\"clavicle_r\"),Weight=0.0007),(Item=(Type=Bone,Name=\"deform_pec_l\"),Weight=0.0025),(Item=(Type=Bone,Name=\"deform_pec_r\"),Weight=0.0025),(Item=(Type=Bone,Name=\"spine_01\"),Weight=0.0073 ...(truncated)"
    Weight=1
   )
  ]}
 New Function_ContainedGraph {nodes [
   RigVMFunctionEntryNode()
   > RigVMFunction_Sequence = RigVMAggregateNode()
   &RigUnit_ParentConstraint(
    AdvancedSettings=(InterpolationType=Average,RotationOrderForFilter=XZY)
    Child=(Type=Bone,Name="neck_01_get_rot")
    Parents=((Item=(Type=Bone,Name="neck_01"),Weight=1))
    Weight=1
    ExecutePin=@RigVMFunction_Sequence.A
   )
   > get_swing_axis_front = RigVMCollapseNode(
    Axis to Get=(X=0,Y=1,Z=1)
    Backward=false
    Forward=true
    Inward=false
    Outward=false
    Rotation Order=ZXY
    Twist Axis=(X=1,Y=0,Z=0)
    Twist_Inward=false
    Twist_Outward=false
    input joint=neck_01_get_rot
   )
   > &RigUnit_ModifyTransforms(
    ItemToModify=((Item=(Type=Bone,Name="neck_01_set_rot"),Transform=(Rotation=(X=0,Y=0,Z=0,W=1),Translation=(X=0,Y=0,Z=0),Scale3D=(X=1,Y=1,Z=1))))
    Weight=1
    WeightMinimum=0
    ItemToModify.0.Transform.Rotation=@get_swing_axis_front.Output Rotate.Value
   )
   &RigUnit_ParentConstraint(
    AdvancedSettings=(InterpolationType=Average,RotationOrderForFilter=XZY)
    Child=(Type=Bone,Name="base_fur_01_lf")
    Parents=((Item=(Type=Bone,Name="clavicle_l"),Weight=1),(Item=(Type=Bone,Name="neck_01_set_rot"),Weight=1))
    Weight=1
    ExecutePin=@RigVMFunction_Sequence.B
   )
   > &RigUnit_ParentConstraint(
    AdvancedSettings=(InterpolationType=Average,RotationOrderForFilter=XZY)
    Child=(Type=Bone,Name="base_fur_02_lf")
    Parents=((Item=(Type=Bone,Name="clavicle_l"),Weight=1),(Item=(Type=Bone,Name="neck_01_set_rot"),Weight=1))
    Weight=1
   )
   > &RigUnit_ParentConstraint(
    AdvancedSettings=(InterpolationType=Average,RotationOrderForFilter=XZY)
    Child=(Type=Bone,Name="base_fur_03_lf")
    Parents=((Item=(Type=Bone,Name="clavicle_l"),Weight=1),(Item=(Type=Bone,Name="neck_01_set_rot"),Weight=1))
    Weight=1
   )
   > &RigUnit_ParentConstraint(
    AdvancedSettings=(InterpolationType=Average,RotationOrderForFilter=XZY)
    Child=(Type=Bone,Name="base_fur_01_cr")
    Parents=((Item=(Type=Bone,Name="neck_01_set_rot"),Weight=1))
    Weight=1
   )
   > &RigUnit_ParentConstraint(
    AdvancedSettings=(InterpolationType=Average,RotationOrderForFilter=XZY)
    Child=(Type=Bone,Name="base_fur_01_rt")
    Parents=((Item=(Type=Bone,Name="clavicle_r"),Weight=1),(Item=(Type=Bone,Name="neck_01_set_rot"),Weight=1))
    Weight=1
   )
   > &RigUnit_ParentConstraint(
    AdvancedSettings=(InterpolationType=Average,RotationOrderForFilter=XZY)
    Child=(Type=Bone,Name="base_fur_02_rt")
    Parents=((Item=(Type=Bone,Name="clavicle_r"),Weight=1),(Item=(Type=Bone,Name="neck_01_set_rot"),Weight=1))
    Weight=1
   )
   > &RigUnit_ParentConstraint(
    AdvancedSettings=(InterpolationType=Average,RotationOrderForFilter=XZY)
    Child=(Type=Bone,Name="base_fur_03_rt")
    Parents=((Item=(Type=Bone,Name="clavicle_r"),Weight=1),(Item=(Type=Bone,Name="neck_01_set_rot"),Weight=1))
    Weight=1
   )
   n1 = /Script/RigVM.RigVMFunction_MathBoolMake(Value=false) {is=struct RigVMUnitNode}
   Spherical Collision_3 = RigVMFunctionReferenceNode(
    Max Radius=10
    Min Radius=10
    Source=(Type=Bone,Name="dyn_fur_01_lf")
    Source Offset=(X=0,Y=0,Z=0)
    Target=(Type=Bone,Name="head")
    Target Offset=(X=2,Y=2,Z=0)
    Weight=1
    Debug=@n1.Value
    ExecuteContext=@RigVMFunction_Sequence.C
   )
   > Spherical Collision_4 = RigVMFunctionReferenceNode(
    Max Radius=20
    Min Radius=12
    Source=(Type=Bone,Name="dyn_fur_02_lf")
    Source Offset=(X=0,Y=0,Z=0)
    Target=(Type=Bone,Name="head")
    Target Offset=(X=2,Y=2,Z=0)
    Weight=1
    Debug=@n1.Value
   )
   > Spherical Collision_5 = RigVMFunctionReferenceNode(
    Max Radius=20
    Min Radius=12
    Source=(Type=Bone,Name="dyn_fur_03_lf")
    Source Offset=(X=0,Y=0,Z=0)
    Target=(Type=Bone,Name="head")
    Target Offset=(X=2,Y=2,Z=0)
    Weight=1
    Debug=@n1.Value
   )
   > Spherical Collision_0 = RigVMFunctionReferenceNode(
    Max Radius=20
    Min Radius=10
    Source=(Type=Bone,Name="dyn_fur_01_cr")
    Source Offset=(X=0,Y=0,Z=0)
    Target=(Type=Bone,Name="head")
    Target Offset=(X=2,Y=2,Z=0)
    Weight=1
    Debug=@n1.Value
   )
   > Spherical Collision_6 = RigVMFunctionReferenceNode(
    Max Radius=20
    Min Radius=10
    Source=(Type=Bone,Name="dyn_fur_01_rt")
    Source Offset=(X=0,Y=0,Z=0)
    Target=(Type=Bone,Name="head")
    Target Offset=(X=2,Y=2,Z=0)
    Weight=1
    Debug=@n1.Value
   )
   > Spherical Collision_7 = RigVMFunctionReferenceNode(
    Max Radius=20
    Min Radius=12
    Source=(Type=Bone,Name="dyn_fur_02_rt")
    Source Offset=(X=0,Y=0,Z=0)
    Target=(Type=Bone,Name="head")
    Target Offset=(X=2,Y=2,Z=0)
    Weight=1
    Debug=@n1.Value
   )
   > Spherical Collision_8 = RigVMFunctionReferenceNode(
    Max Radius=20
    Min Radius=12
    Source=(Type=Bone,Name="dyn_fur_03_rt")
    Source Offset=(X=0,Y=0,Z=0)
    Target=(Type=Bone,Name="head")
    Target Offset=(X=2,Y=2,Z=0)
    Weight=1
    Debug=@n1.Value
   )
   DeltaRotation = RigVMFunctionReferenceNode(
    Rotation Order=ZYX
    child=clavicle_l
    debugRotation=false
    max=-60
    onlyTwist=false
    parent=clavicle_l_null
    parentType=Null
    premult=1
    primAxis=1
    twistAxis=(X=1,Y=0,Z=0)
    ExecuteContext=@RigVMFunction_Sequence.D
   )
   > &RigUnit_ModifyTransforms(
    ItemToModify="((Item=(Type=Bone,Name=\"base_fur_01_lf\"),Transform=(Rotation=(X=0,Y=-0.3827,Z=0,W=0.9239),Translation=(X=0,Y=0,Z=0),Scale3D=(X=1,Y=1,Z=1))),(Item=(Type=Bone,Name=\"base_fur_02_lf\"),Transform=(Rotation=(X=0,Y=-0.6088,Z=0,W=0.7934),Translation=(X=12,Y=0,Z=3), ...(truncated)"
    WeightMinimum=0
    Weight=@DeltaRotation.outputWeight
   )
   > DeltaRotation_3 = RigVMFunctionReferenceNode(
    Rotation Order=ZYX
    child=clavicle_r
    debugRotation=false
    max=-60
    onlyTwist=false
    parent=clavicle_r_null
    parentType=Null
    premult=1
    primAxis=1
    twistAxis=(X=1,Y=0,Z=0)
   )
   > &RigUnit_ModifyTransforms(
    ItemToModify="((Item=(Type=Bone,Name=\"base_fur_01_rt\"),Transform=(Rotation=(X=0,Y=-0.3827,Z=0,W=0.9239),Translation=(X=0,Y=0,Z=0),Scale3D=(X=1,Y=1,Z=1))),(Item=(Type=Bone,Name=\"base_fur_02_rt\"),Transform=(Rotation=(X=0,Y=-0.6088,Z=0,W=0.7934),Translation=(X=12,Y=0,Z=3), ...(truncated)"
    WeightMinimum=0
    Weight=@DeltaRotation_3.outputWeight
   )
   > DeltaRotation_2 = RigVMFunctionReferenceNode(
    Rotation Order=ZYX
    child=clavicle_l
    debugRotation=false
    max=60
    onlyTwist=false
    parent=clavicle_l_null
    parentType=Null
    premult=1
    primAxis=2
    twistAxis=(X=1,Y=0,Z=0)
   )
   > &RigUnit_ModifyTransforms(
    ItemToModify=((Item=(Type=Bone,Name="base_fur_01_lf"),Transform=(Rotation=(X=0,Y=-0.1908,Z=0,W=0.9816),Translation=(X=-2.4,Y=0,Z=1.6),Scale3D=(X=1,Y=1,Z=1))))
    WeightMinimum=0
    Weight=@DeltaRotation_2.outputWeight
   )
   > DeltaRotation_1 = RigVMFunctionReferenceNode(
    Rotation Order=ZYX
    child=clavicle_r
    debugRotation=false
    max=60
    onlyTwist=false
    parent=clavicle_r_null
    parentType=Null
    premult=1
    primAxis=2
    twistAxis=(X=1,Y=0,Z=0)
   )
   > &RigUnit_ModifyTransforms(
    ItemToModify=((Item=(Type=Bone,Name="base_fur_01_rt"),Transform=(Rotation=(X=0,Y=-0.1908,Z=0,W=0.9816),Translation=(X=-2.4,Y=0,Z=1.6),Scale3D=(X=1,Y=1,Z=1))))
    WeightMinimum=0
    Weight=@DeltaRotation_1.outputWeight
   )
   > n2 = &RigUnit_SphericalPoseReader(
    ActiveRegionScaleFactors=(PositiveWidth=1,NegativeWidth=1,PositiveHeight=1,NegativeHeight=1)
    ActiveRegionSize=0.1
    Debug=(bDrawDebug=false,bDraw2D=False,bDrawLocalAxes=False,DebugScale=25,DebugSegments=20,DebugThickness=0.25)
    DriverAxis=(X=-1,Y=0,Z=0)
    DriverItem=(Type=Bone,Name="upperarm_r")
    FalloffRegionScaleFactors=(PositiveWidth=1,NegativeWidth=1,PositiveHeight=1,NegativeHeight=1)
    FalloffSize=0.3
    FlipHeightScaling=False
    FlipWidthScaling=False
    OptionalParentItem=(Type=Bone,Name="None")
    RotationOffset=(X=-90,Y=0,Z=10)
   )
   > &RigUnit_ModifyTransforms(
    ItemToModify="((Item=(Type=Bone,Name=\"base_fur_01_rt\"),Transform=(Rotation=(X=0.212,Y=0.0175,Z=0.0085,W=-0.9771),Translation=(X=-0.0766,Y=2.4433,Z=2.9624),Scale3D=(X=1,Y=1,Z=1))),(Item=(Type=Bone,Name=\"base_fur_02_rt\"),Transform=(Rotation=(X=0,Y=-0.3619,Z=0,W=-0.9322),T ...(truncated)"
    WeightMinimum=0
    Weight=@n2.OutputParam
   )
   > DeltaRotation_4 = RigVMFunctionReferenceNode(
    Rotation Order=ZYX
    child=head
    debugRotation=false
    max=-50
    min=0
    onlyTwist=false
    parent=head_null
    parentType=Null
    premult=1
    primAxis=2
    twistAxis=(X=1,Y=0,Z=0)
   )
   > &RigUnit_ModifyTransforms(
    ItemToModify="((Item=(Type=Bone,Name=\"dyn_fur_01_cr\"),Transform=(Rotation=(X=0,Y=-0.2206,Z=0,W=-0.9754),Translation=(X=0,Y=0,Z=0),Scale3D=(X=1,Y=1,Z=1))),(Item=(Type=Bone,Name=\"dyn_fur_03_lf\"),Transform=(Rotation=(X=0.1332,Y=0.2282,Z=-0.0177,W=0.9643),Translation=(X=1.1 ...(truncated)"
    WeightMinimum=0
    Weight=@DeltaRotation_4.outputWeight
   )
   n3 = &RigVMFunction_MathDoubleMul(B=RigVMVariableNode() {variable=isCrouchingCR}.Value, A=@DeltaRotation_4.outputWeight)
   > &RigUnit_ModifyTransforms(
    ItemToModify="((Item=(Type=Bone,Name=\"dyn_fur_01_rt\"),Transform=(Rotation=(X=0,Y=-0.0369,Z=0,W=0.9993),Translation=(X=0,Y=0,Z=0),Scale3D=(X=1,Y=1,Z=1))),(Item=(Type=Bone,Name=\"dyn_fur_03_rt\"),Transform=(Rotation=(X=-0.0442,Y=0.1783,Z=0.008,W=0.9829),Translation=(X=0,Y=0 ...(truncated)"
    WeightMinimum=0
    Weight=@n3.Result
   )
   > &RigUnit_ModifyTransforms(
    ItemToModify="((Item=(Type=Bone,Name=\"dyn_fur_01_cr\"),Transform=(Rotation=(X=0,Y=-0.2201,Z=0,W=0.9755),Translation=(X=0,Y=0,Z=0),Scale3D=(X=1,Y=1,Z=1))),(Item=(Type=Bone,Name=\"dyn_fur_03_rt\"),Transform=(Rotation=(X=0.1526,Y=-0.0677,Z=-0.0105,W=0.9859),Translation=(X=0,Y ...(truncated)"
    WeightMinimum=0
    Weight=&RigVMFunction_MathDoubleMul(A=@n3.Result, B=RigVMVariableNode() {variable=isCrouchingMoveCR}.Value).Result
   )
   > n4 = &RigUnit_SphericalPoseReader(
    ActiveRegionScaleFactors=(PositiveWidth=1,NegativeWidth=1,PositiveHeight=1,NegativeHeight=1)
    ActiveRegionSize=0.03
    Debug=(bDrawDebug=false,bDraw2D=False,bDrawLocalAxes=False,DebugScale=25,DebugSegments=20,DebugThickness=0.25)
    DriverAxis=(X=1,Y=0,Z=0)
    DriverItem=(Type=Bone,Name="clavicle_l")
    FalloffRegionScaleFactors=(PositiveWidth=1,NegativeWidth=4,PositiveHeight=4,NegativeHeight=4)
    FalloffSize=0.06
    FlipHeightScaling=False
    FlipWidthScaling=False
    OptionalParentItem=(Type=Bone,Name="None")
    RotationOffset=(X=0,Y=-50,Z=8)
   )
   Multiply = &RigVMFunction_MathDoubleMul(B=@n4.OutputParam, A=@DeltaRotation_4.outputWeight)
   > &RigUnit_ModifyTransforms(
    ItemToModify="((Item=(Type=Bone,Name=\"dyn_fur_02_rt\"),Transform=(Rotation=(X=0,Y=0.1001,Z=0,W=-0.995),Translation=(X=0,Y=0,Z=0),Scale3D=(X=1,Y=1,Z=1))),(Item=(Type=Bone,Name=\"dyn_fur_03_lf\"),Transform=(Rotation=(X=0.2081,Y=-0.1222,Z=-0.0052,W=0.9704),Translation=(X=0,Y= ...(truncated)"
    WeightMinimum=0
    Weight=&RigVMFunction_MathDoubleMul(A=@Multiply.Result, B=@n3.Result).Result
   )
   > n5 = &RigUnit_SphericalPoseReader(
    ActiveRegionScaleFactors=(PositiveWidth=1,NegativeWidth=1,PositiveHeight=1,NegativeHeight=1)
    ActiveRegionSize=0.035
    Debug=(bDrawDebug=false,bDraw2D=False,bDrawLocalAxes=False,DebugScale=25,DebugSegments=20,DebugThickness=0.25)
    DriverAxis=(X=1,Y=0,Z=0)
    DriverItem=(Type=Bone,Name="clavicle_l")
    FalloffRegionScaleFactors=(PositiveWidth=4,NegativeWidth=4,PositiveHeight=4,NegativeHeight=4)
    FalloffSize=0.07
    FlipHeightScaling=False
    FlipWidthScaling=False
    OptionalParentItem=(Type=Bone,Name="None")
    RotationOffset=(X=0,Y=-40,Z=-10)
   )
   > &RigUnit_ModifyTransforms(
    ItemToModify="((Item=(Type=Bone,Name=\"dyn_fur_01_cr\"),Transform=(Rotation=(X=0,Y=-0.1166,Z=0,W=0.9932),Translation=(X=0,Y=0,Z=0),Scale3D=(X=1,Y=1,Z=1))),(Item=(Type=Bone,Name=\"dyn_fur_03_rt\"),Transform=(Rotation=(X=0,Y=-0.0448,Z=0,W=0.999),Translation=(X=0,Y=0,Z=0),Scal ...(truncated)"
    WeightMinimum=0
    Weight=&RigVMFunction_MathDoubleMul(A=@n5.OutputParam, B=RigVMVariableNode() {variable=isCrouchingMoveCR}.Value).Result
   )
   > RigVMFunctionReturnNode()
   &RigUnit_ModifyTransforms(
    ExecutePin=()
    ItemToModify="((Item=(Type=Bone,Name=\"dyn_fur_01_cr\"),Transform=(Rotation=(X=0,Y=0.0659,Z=0,W=0.9978),Translation=(X=-0.1156,Y=0,Z=2.5794),Scale3D=(X=1,Y=1,Z=1))),(Item=(Type=Bone,Name=\"dyn_fur_03_lf\"),Transform=(Rotation=(X=-0.0929,Y=-0.2395,Z=-0.0456,W=0.9654),Transla ...(truncated)"
    WeightMinimum=0
    Weight=@Multiply.Result
   )
   RigVMRerouteNode(Value=@n3.Result)
  ]}
 New Function_ContainedGraph {nodes [
   n1 = RigVMFunctionEntryNode()
   RigUnit_GetTransform = &RigUnit_GetTransform(Item=(Type=Bone,Name="thigh_l_for_cons"), Space=LocalSpace, bInitial=false, Item.Name=@n1.input joint)
   RigVMFunction_MathQuaternionToEuler = &RigVMFunction_MathQuaternionToEuler(
    RotationOrder=@n1.Rotation Order
    Value=&RigVMFunction_MathQuaternionSwingTwist(
     TwistAxis=(X=1,Y=0,Z=0)
     TwistAxis.X=@n1.Twist Axis.X
     TwistAxis.Y=@n1.Twist Axis.Y
     TwistAxis.Z=@n1.Twist Axis.Z
     Input=@RigUnit_GetTransform.Transform.Rotation
    ).Swing
   )
   n2 = &RigVMFunction_MathDoubleMul(B=@RigVMFunction_MathQuaternionToEuler.Result.Z, A=@n1.Axis to Get.Z)
   n3 = &RigVMFunction_MathDoubleLess(B=0, A=@n2.Result) {template="Less::Execute(in A,in B,out Result)"}
   n4 = &RigVMFunction_MathDoubleGreater(B=0, A=@n2.Result) {template="Greater::Execute(in A,in B,out Result)"}
   Add = &RigVMFunction_MathDoubleAdd(
    A=RigVMDispatchNode(
     False=0
     Condition=@n1.Forward
     True=RigVMDispatchNode(False=0, Condition=@n3.Result, True=@RigVMFunction_MathQuaternionToEuler.Result.Z) {template="DISPATCH_RigVMDispatch_If(in Condition,in True,in False,out Result)"}.Result
    ) {template="DISPATCH_RigVMDispatch_If(in Condition,in True,in False,out Result)"}.Result
    B=RigVMDispatchNode(
     False=0
     Condition=@n1.Backward
     True=RigVMDispatchNode(False=0, Condition=@n4.Result, True=@RigVMFunction_MathQuaternionToEuler.Result.Z) {template="DISPATCH_RigVMDispatch_If(in Condition,in True,in False,out Result)"}.Result
    ) {template="DISPATCH_RigVMDispatch_If(in Condition,in True,in False,out Result)"}.Result
   )
   n5 = &RigVMFunction_MathDoubleMul(A=@n1.Axis to Get.Y, B=@RigVMFunction_MathQuaternionToEuler.Result.Y)
   Less = &RigVMFunction_MathDoubleLess(B=0, A=@n5.Result) {template="Less::Execute(in A,in B,out Result)"}
   n6 = RigVMDispatchNode(False=0, Condition=@Less.Result, True=@RigVMFunction_MathQuaternionToEuler.Result.Y) {template="DISPATCH_RigVMDispatch_If(in Condition,in True,in False,out Result)"}
   Greater = &RigVMFunction_MathDoubleGreater(B=0, A=@n5.Result) {template="Greater::Execute(in A,in B,out Result)"}
   n7 = RigVMDispatchNode(False=0, Condition=@Greater.Result, True=@RigVMFunction_MathQuaternionToEuler.Result.Y) {template="DISPATCH_RigVMDispatch_If(in Condition,in True,in False,out Result)"}
   RigVMFunction_MathQuaternionFromEuler = &RigVMFunction_MathQuaternionFromEuler(
    Euler=(X=0,Y=0,Z=0)
    Euler.Z=@Add.Result
    Euler.Y=&RigVMFunction_MathDoubleAdd(
     A=RigVMDispatchNode(False=0, Condition=@n1.Outward, True=@n6.Result) {template="DISPATCH_RigVMDispatch_If(in Condition,in True,in False,out Result)"}.Result
     B=RigVMDispatchNode(False=0, Condition=@n1.Inward, True=@n7.Result) {template="DISPATCH_RigVMDispatch_If(in Condition,in True,in False,out Result)"}.Result
    ).Result
    RotationOrder=@n1.Rotation Order
   )
   > RigVMFunctionReturnNode(Output Rotate=(Value=(X=0,Y=0,Z=0,W=1)), Output Rotate.Value=@RigVMFunction_MathQuaternionFromEuler.Result)
  ]}
 New Function_ContainedGraph {nodes [
   n1 = RigVMFunctionEntryNode()
   RigUnit_GetTransform = &RigUnit_GetTransform(Item=(Type=Bone,Name="spine_02_for_cons_offset"), Space=LocalSpace, bInitial=false, Item.Name=@n1.Input Bone)
   RigVMFunction_MathQuaternionToEuler = &RigVMFunction_MathQuaternionToEuler(RotationOrder=ZYX, Value=@RigUnit_GetTransform.Transform.Rotation)
   RigVMFunction_MathQuaternionFromEuler = &RigVMFunction_MathQuaternionFromEuler(Euler=&RigVMFunction_MathVectorMul(B=@n1.Axis, A=@RigVMFunction_MathQuaternionToEuler.Result) {template="Multiply::Execute(in A,in B,out Result)"}.Result)
   > Set Transform_1 = &RigUnit_SetRotation(
    Item=(Type=Bone,Name="reverse_spine_02")
    Weight=1
    bInitial=false
    Item.Name=@n1.Target Bone
    Value=@RigVMFunction_MathQuaternionFromEuler.Result
   )
   > RigVMFunctionReturnNode()
   Set Transform = &RigUnit_SetRotation(ExecutePin=(), Item=(Type=Bone,Name="reverse_spine_02"), Value=(X=0,Y=0,Z=0,W=1), Weight=1, bInitial=false)
  ]}
 New Function_ContainedGraph {nodes [
   RigVMFunctionEntryNode()
   > &RigUnit_ParentConstraint(
    AdvancedSettings=(InterpolationType=Average,RotationOrderForFilter=XZY)
    Child=(Type=Bone,Name="spine_02_for_cons")
    Parents=((Item=(Type=Bone,Name="spine_02"),Weight=1))
    Weight=1
   )
   > reverse rotation = RigVMFunctionReferenceNode(Axis=(X=-0.1,Y=-0.5,Z=-0.5), Input Bone=spine_02_for_cons, Target Bone=offset_main_jacket_01_cr)
   > &RigUnit_ParentConstraint(
    AdvancedSettings=(InterpolationType=Average,RotationOrderForFilter=XZY)
    Child=(Type=Bone,Name="spine_03_for_cons")
    Parents=((Item=(Type=Bone,Name="spine_03"),Weight=1))
    Weight=1
   )
   > reverse rotation_1 = RigVMFunctionReferenceNode(Axis=(X=-0.3,Y=-1,Z=-1), Input Bone=spine_03_for_cons, Target Bone=dyn_main_jacket_01_cr)
   > n1 = &RigUnit_SphericalPoseReader(
    ActiveRegionScaleFactors=(PositiveWidth=1,NegativeWidth=1,PositiveHeight=1,NegativeHeight=1)
    ActiveRegionSize=0.2
    Debug=(bDrawDebug=false,bDraw2D=False,bDrawLocalAxes=False,DebugScale=25,DebugSegments=20,DebugThickness=0.25)
    DriverAxis=(X=1,Y=0,Z=0)
    DriverItem=(Type=Bone,Name="upperarm_l")
    FalloffRegionScaleFactors=(PositiveWidth=1,NegativeWidth=1,PositiveHeight=1,NegativeHeight=1)
    FalloffSize=0.4
    FlipHeightScaling=False
    FlipWidthScaling=False
    OptionalParentItem=(Type=Bone,Name="None")
    RotationOffset=(X=90,Y=0,Z=-20)
   )
   RigVMFunction_AnimEvalRichCurve = &RigVMFunction_AnimEvalRichCurve(
    Curve="(EditorCurveData=(Keys=((InterpMode=RCIM_Cubic,TangentMode=RCTM_Auto,TangentWeightMode=RCTWM_WeightedNone,Time=0,Value=0,ArriveTangent=0,ArriveTangentWeight=0,LeaveTangent=0,LeaveTangentWeight=0),(InterpMode=RCIM_Cubic,TangentMode=RCTM_Auto,TangentWeightMo ...(truncated)"
    SourceMinimum=0
    TargetMinimum=0
    Value=@n1.OutputParam
   )
   AlphaInterp = &RigVMFunction_AlphaInterp(InterpSpeedDecreasing=10, InterpSpeedIncreasing=10, Scale=1, bClampResult=False, Value=@RigVMFunction_AnimEvalRichCurve.Result)
   > &RigUnit_ModifyTransforms(
    ItemToModify=((Item=(Type=Bone,Name="base_jacket_helper_lf"),Transform=(Rotation=(X=0,Y=0,Z=0,W=1),Translation=(X=100,Y=0,Z=0),Scale3D=(X=1,Y=1,Z=1))))
    WeightMinimum=0
    Weight=@AlphaInterp.Result
   )
   > n2 = &RigUnit_SphericalPoseReader(
    ActiveRegionScaleFactors=(PositiveWidth=1,NegativeWidth=1,PositiveHeight=1,NegativeHeight=1)
    ActiveRegionSize=0.2
    Debug=(bDrawDebug=false,bDraw2D=False,bDrawLocalAxes=False,DebugScale=25,DebugSegments=20,DebugThickness=0.25)
    DriverAxis=(X=-1,Y=0,Z=0)
    DriverItem=(Type=Bone,Name="upperarm_r")
    FalloffRegionScaleFactors=(PositiveWidth=1,NegativeWidth=1,PositiveHeight=1,NegativeHeight=1)
    FalloffSize=0.4
    FlipHeightScaling=False
    FlipWidthScaling=False
    OptionalParentItem=(Type=Bone,Name="None")
    RotationOffset=(X=-90,Y=0,Z=-20)
   )
   n3 = &RigVMFunction_AnimEvalRichCurve(
    Curve="(EditorCurveData=(Keys=((InterpMode=RCIM_Cubic,TangentMode=RCTM_Auto,TangentWeightMode=RCTWM_WeightedNone,Time=0,Value=0,ArriveTangent=0,ArriveTangentWeight=0,LeaveTangent=0,LeaveTangentWeight=0),(InterpMode=RCIM_Cubic,TangentMode=RCTM_Auto,TangentWeightMo ...(truncated)"
    SourceMinimum=0
    TargetMinimum=0
    Value=@n2.OutputParam
   )
   n4 = &RigVMFunction_AlphaInterp(InterpSpeedDecreasing=10, InterpSpeedIncreasing=10, Scale=1, bClampResult=False, Value=@n3.Result)
   > &RigUnit_ModifyTransforms(
    ItemToModify=((Item=(Type=Bone,Name="dyn_jacket_helper_rt"),Transform=(Rotation=(X=0,Y=0,Z=0,W=1),Translation=(X=50,Y=0,Z=0),Scale3D=(X=1,Y=1,Z=1))))
    WeightMinimum=0
    Weight=@n4.Result
   )
   Cast = &RigVMFunction_MathBoolToFloat(Value=RigVMVariableNode() {variable=isCrouching}.Value)
   n5 = &RigVMFunction_AnimEvalRichCurve(
    Curve=(EditorCurveData=(Keys=((),(Time=0.5,Value=1),(Time=0.8))))
    SourceMaximum=2
    SourceMinimum=0
    TargetMinimum=0
    Value=&RigVMFunction_MathDoubleRemap(
     SourceMaximum=65
     SourceMinimum=45
     TargetMaximum=1
     TargetMinimum=0
     Value=&RigUnit_GetRelativeTransformForItem(Child=(Type=Bone,Name="pelvis"), Parent=(Type=Bone,Name="root"), bChildInitial=False, bParentInitial=False).RelativeTransform.Translation.Z
    ).Result
   )
   Multiply = &RigVMFunction_MathDoubleMul(B=@Cast.Result, A=@n5.Result)
   > &RigUnit_ModifyTransforms(
    ItemToModify=((Item=(Type=Bone,Name="base_jacket_01_lf"),Transform=(Rotation=(X=0,Y=0,Z=0,W=1),Translation=(X=-3,Y=0,Z=0),Scale3D=(X=1,Y=1,Z=1))))
    WeightMinimum=0
    Weight=@Multiply.Result
   )
   > RigVMVariableNode(Value=@Multiply.Result) {variable=isCrouchingMoveCR}
   > RigVMFunctionReturnNode()
  ]}
 New Function_ContainedGraph {nodes [
   RigVMFunctionEntryNode()
   > &RigUnit_ParentConstraint(
    AdvancedSettings=(RotationOrderForFilter=XZY)
    Child=(Type=Bone,Name="base_bullet_01_rt")
    Parents="((Item=(Type=Bone,Name=\"thigh_twist_01_r\"),Weight=0.0025),(Item=(Type=Bone,Name=\"deform_glute_r\"),Weight=0.0155),(Item=(Type=Bone,Name=\"spine_03\"),Weight=0.0289),(Item=(Type=Bone,Name=\"spine_01\"),Weight=0.0421),(Item=(Type=Bone,Name=\"spine_02\"),Weight=0.04 ...(truncated)"
    Weight=1
   )
   > &RigUnit_ParentConstraint(
    AdvancedSettings=(RotationOrderForFilter=XZY)
    Child=(Type=Bone,Name="base_bullet_02_rt")
    Parents="((Item=(Type=Bone,Name=\"thigh_twist_01_r\"),Weight=0.0057),(Item=(Type=Bone,Name=\"deform_glute_r\"),Weight=0.0156),(Item=(Type=Bone,Name=\"spine_03\"),Weight=0.0348),(Item=(Type=Bone,Name=\"spine_01\"),Weight=0.0518),(Item=(Type=Bone,Name=\"spine_02\"),Weight=0.05 ...(truncated)"
    Weight=1
   )
   > &RigUnit_ParentConstraint(
    AdvancedSettings=(RotationOrderForFilter=XZY)
    Child=(Type=Bone,Name="base_bullet_03_rt")
    Parents="((Item=(Type=Bone,Name=\"thigh_twist_01_r\"),Weight=0.0057),(Item=(Type=Bone,Name=\"deform_glute_r\"),Weight=0.0156),(Item=(Type=Bone,Name=\"spine_03\"),Weight=0.0348),(Item=(Type=Bone,Name=\"spine_01\"),Weight=0.0518),(Item=(Type=Bone,Name=\"spine_02\"),Weight=0.05 ...(truncated)"
    Weight=1
   )
   > &RigUnit_ParentConstraint(
    AdvancedSettings=(RotationOrderForFilter=XZY)
    Child=(Type=Bone,Name="base_zipper")
    Parents="((Item=(Type=Bone,Name=\"dyn_jacket_02_rt\"),Weight=0.0032),(Item=(Type=Bone,Name=\"spine_03\"),Weight=0.0065),(Item=(Type=Bone,Name=\"dyn_jacket_02_rt_end\"),Weight=0.0147),(Item=(Type=Bone,Name=\"dyn_jacket_01_rt\"),Weight=0.0922),(Item=(Type=Bone,Name=\"dyn_jack ...(truncated)"
    Weight=1
   )
   > RigVMFunctionReturnNode()
  ]}
 RigVMFunctionLibrary {nodes [
   DeltaRotation = RigVMCollapseNode(ExecuteContext=(), Rotation Order=ZYX, premult=1, twistAxis=(X=1,Y=0,Z=0))
   back strap = RigVMCollapseNode(ExecuteContext=())
   chain = RigVMCollapseNode(ExecuteContext=())
   fur = RigVMCollapseNode(ExecuteContext=())
   jacket = RigVMCollapseNode(ExecuteContext=())
   necklace = RigVMCollapseNode(ExecuteContext=())
   pistol = RigVMCollapseNode(ExecuteContext=())
   reverse rotation = RigVMCollapseNode(Axis=(X=0,Y=0,Z=0), ExecuteContext=())
   surface constraints bullets and zipper = RigVMCollapseNode(ExecuteContext=())
  ]}
 RigVMFunction_Sequence_1_ContainedGraph {nodes [
   n1 = RigVMFunctionEntryNode()
   RigVMFunction_Sequence = &RigVMFunction_Sequence(ExecuteContext=@n1.ExecuteContext)
   n2 = &RigVMFunction_Sequence(ExecuteContext=@RigVMFunction_Sequence.B)
   n3 = &RigVMFunction_Sequence(ExecuteContext=@n2.B)
   RigVMFunctionReturnNode(A=@RigVMFunction_Sequence.A, B=@n2.A, C=@n3.A, D=@n3.B)
  ]}
 RigVMModel {nodes [
   RigUnit_BeginExecution = &RigUnit_BeginExecution()
   FN_DeformRig_FNC = RigVMFunctionReferenceNode(
    DeformRig_ Curve="(deformRig_l_upperarm_up_alpha_72_#1=False,deformRig_r_upperarm_up_alpha_73_#2=False,deformRig_l_upperarm_fr_alpha_74_#3=False,deformRig_r_upperarm_fr_alpha_75_CB51B8 ...(truncated)"
    DeformRig_Bool="(upperarm_1_#4=True,elbow_3_#5=True,wrist_7_#6=True,groin_8_#7=True,glute_17_#8=True,knee_9_40C408324CA261 ...(truncated)"
    DeformRig_Debug=(upperarm_debug_18_#4=False,wrist_debug_20_#6=False,lat_debug_26_#9=False,groinGlute_debug_27_#7=False)
    DeformRig_Value="(upperarm_up_18_#10=(Rotation=(X=0.0,Y=-0.0,Z=0.0,W=1.0),Translation=(X=0.0,Y=0.0,Z=8.0),Scale3D=(X=0.70,Y=1.0,Z=1.0)),upperarm_fwd_fr_19_#11=(Rotation=(X=0,Y=0,Z=0,W=1),Translation=(X=0.0,Y=10.0,Z= ...(truncated)"
    ExecuteContext=@RigUnit_BeginExecution.ExecutePin
   )
   > surface constraints bullets and zipper = RigVMFunctionReferenceNode()
   > pistol = RigVMFunctionReferenceNode()
   > back strap = RigVMFunctionReferenceNode()
   > chain = RigVMFunctionReferenceNode()
   > necklace = RigVMFunctionReferenceNode()
   > fur = RigVMFunctionReferenceNode()
   > jacket = RigVMFunctionReferenceNode()
   &RigUnit_GetRelativeTransformForItem(Child=(Type=Bone,Name="None"), Parent=(Type=Bone,Name="None"), bChildInitial=False, bParentInitial=False)
   &RigUnit_PrepareForExecution()
   addFKCtrl = RigVMFunctionReferenceNode(
    Custom_FKCtrl=()
    ExecuteContext=()
    Options_FKCtrl="(pelvis_1_#12=false,spine_3_#13=false,clavicle_5_#14=false,arm_7_#15=false,leg_9_#16=true,head_11_46DC15EF465CD ...(truncated)"
    Shape_Settings="(pelvis_2_#17=(bVisible=True,Name=\"Hexagon_Thin\",Color=(R=0.0117,G=0.8281,B=1,A=1),Transform=(Rotation=(X=0,Y=-0.7071,Z=0,W=0.7071),Translation=(X=2,Y=0,Z=0),Scale3D=(X=5,Y=5,Z=7))),spine_4_#18=(bVi ...(truncated)"
   )
   setFKCtrl = RigVMFunctionReferenceNode(ExecuteContext=(), setFK_Ctrl=false)
  ]}
]
hierarchy [
 root {loc=<0 0 0>, rot=<0 0 0 1>}
 pelvis {parent=root, loc=<0 -0.7071 93.1191>, rot=<0.0127 0.707 -0.0127 -0.707>}
 spine_01 {parent=pelvis, loc=<6.7494 -0.6388 0>, rot=<0 0 0.034 -0.9994>}
 spine_02 {parent=spine_01, loc=<7.2518 -0.035 0>, rot=<0 0 0.0238 -0.9997>}
 spine_03 {parent=spine_02, loc=<6.0104 0 0>, rot=<0 0 0.0149 -0.9999>}
 spine_04 {parent=spine_03, loc=<8.1024 0 0>, rot=<0 0 -0.0132 -0.9999>}
 spine_05 {parent=spine_04, loc=<10.0146 0 0>, rot=<0 0 -0.0015 -1>}
 neck_01 {parent=spine_05, loc=<11.3138 0.1529 0>, rot=<0 0 -0.2003 -0.9797>}
 neck_02 {parent=neck_01, loc=<4.5627 0 0>, rot=<0 0 0.0504 -0.9987>}
 head {parent=neck_02, loc=<5.3914 0 0>, rot=<0 0 0.115 -0.9934>}
 head_null {parent=spine_05, type=Null, loc=<20.6573 3.5489 0>, rot=<0 0 -0.036 -0.9994>}
 attach_cape {parent=spine_05, loc=<9.017 -11.714 0>, rot=<0.0616 -0.7048 -0.061 -0.7041>}
 cape {parent=attach_cape, type=Null, loc=<0 0 0>, rot=<0 0 0 1>}
 attach_backpack {parent=spine_05, loc=<-0.123 -13.473 0>, rot=<-0.0093 -0.7074 0.0099 -0.7067>}
 Backpack_BR {parent=attach_backpack, type=Null, loc=<0 -1 0>, rot=<0 0 0 1>}
 clavicle_l {parent=spine_05, loc=<4.3575 -0.1466 -2.3223>, rot=<0.0465 -0.7289 -0.009 -0.6829>}
 upperarm_l {parent=clavicle_l, loc=<15.5111 -0.0001 0>, rot=<0.0226 -0.3234 -0.0088 -0.9459>}
 lowerarm_l {parent=upperarm_l, loc=<26.6514 0 0>, rot=<0.0391 -0.0099 -0.2453 -0.9686>}
 hand_l {parent=lowerarm_l, loc=<22.3491 0 0>, rot=<-0.7081 -0.0507 0.0975 -0.6975>}
 thumb_01_l {parent=hand_l, loc=<1.3118 -0.3191 -2.459>, rot=<0.6546 -0.4042 0.0664 -0.6354>}
 thumb_02_l {parent=thumb_01_l, loc=<5.9336 -0.0448 0.0401>, rot=<-0.0037 0.0213 0.298 -0.9543>}
 thumb_03_l {parent=thumb_02_l, loc=<2.2223 0.3682 -0.0907>, rot=<0.0118 -0.013 0.0066 -0.9998>}
 index_metacarpal_l {parent=hand_l, loc=<1.6535 0.0183 -2.96>, rot=<0.0158 -0.0124 -0.0173 -0.9996>}
 index_01_l {parent=index_metacarpal_l, loc=<8.3008 0.0357 -0.0524>, rot=<-0.0044 0.0485 0.0924 -0.9945>}
 index_02_l {parent=index_01_l, loc=<3.3587 -0.3004 0.0452>, rot=<-0.002 0.0027 0.3225 -0.9466>}
 index_03_l {parent=index_02_l, loc=<2.5578 -0.123 -0.0252>, rot=<-0.002 0.0067 0.2558 -0.9667>}
 middle_metacarpal_l {parent=hand_l, loc=<2.1856 -0.1725 -0.5984>, rot=<0.006 0.0168 -0.0329 -0.9993>}
 middle_01_l {parent=middle_metacarpal_l, loc=<7.3275 -0.2447 0.0129>, rot=<-0.0208 0.0042 0.0858 -0.9961>}
 middle_02_l {parent=middle_01_l, loc=<4.0123 0.1401 0.1175>, rot=<-0.0218 0.0258 0.3512 -0.9357>}
 middle_03_l {parent=middle_02_l, loc=<2.6077 0.0534 0.1373>, rot=<0.0082 0.0188 0.1708 -0.9851>}
 ring_metacarpal_l {parent=hand_l, loc=<2.2074 -0.3625 1.3638>, rot=<-0.0284 0.0722 -0.0458 -0.9959>}
 ring_01_l {parent=ring_metacarpal_l, loc=<7.0295 -0.6197 -0.1446>, rot=<-0.0489 -0.0252 0.1892 -0.9804>}
 ring_02_l {parent=ring_01_l, loc=<3.6174 0.4647 -0.1027>, rot=<0 0 0.2806 -0.9598>}
 ring_03_l {parent=ring_02_l, loc=<2.4152 0.114 0.0199>, rot=<0 0 0.1189 -0.9929>}
 pinky_metacarpal_l {parent=hand_l, loc=<1.7616 -0.9753 3.3814>, rot=<-0.0397 0.0814 -0.069 -0.9935>}
 pinky_01_l {parent=pinky_metacarpal_l, loc=<7.1709 -0.4831 -0.1094>, rot=<-0.0654 -0.0035 0.2242 -0.9723>}
 pinky_02_l {parent=pinky_01_l, loc=<2.8274 0.0893 -0.1043>, rot=<0.0037 0.0011 0.2854 -0.9584>}
 pinky_03_l {parent=pinky_02_l, loc=<2.2046 0.0307 -0.0078>, rot=<-0.0002 0.0098 0.1445 -0.9895>}
 deform_wrist_dn_base_l {parent=hand_l, loc=<0 -2.2342 0>, rot=<0 0 0 -1>}
 deform_wrist_dn_l {parent=deform_wrist_dn_base_l, loc=<0 0 0>, rot=<0 0 0 1>}
 deform_wrist_up_base_l {parent=hand_l, loc=<0 2.2682 0>, rot=<0 0 0 -1>}
 deform_wrist_up_l {parent=deform_wrist_up_base_l, loc=<0 0 0>, rot=<0 0 0 1>}
 lowerarm_twist_02_l {parent=lowerarm_l, loc=<9.6199 0 0>, rot=<0 0 0 -1>}
 lowerarm_twist_01_l {parent=lowerarm_l, loc=<17.8549 0 0>, rot=<0 0 0 -1>}
 deform_elbow_in_base_l {parent=lowerarm_l, loc=<0.2112 5.259 0.8274>, rot=<0 0 0 -1>}
 deform_elbow_in_l {parent=deform_elbow_in_base_l, loc=<0 0 0>, rot=<0 0 0 1>}
 deform_elbow_out_base_l {parent=lowerarm_l, loc=<0 -4 0>, rot=<0 0 0 -1>}
 deform_elbow_out_l {parent=deform_elbow_out_base_l, loc=<0 0 0>, rot=<0 0 0 1>}
 upperarm_twist_01_l {parent=upperarm_l, loc=<0.0041 0 0>, rot=<0 0 0 -1>}
 deform_upperarm_bk_base_l {parent=upperarm_twist_01_l, loc=<1.0813 -7.2192 -1.9453>, rot=<0 0 0 -1>}
 deform_upperarm_bk_l {parent=deform_upperarm_bk_base_l, loc=<0 0 0>, rot=<0 0 0 1>}
 deform_upperarm_fr_base_l {parent=upperarm_twist_01_l, loc=<-0.3691 7.0507 -2.765>, rot=<0 0 0 -1>}
 deform_upperarm_fr_l {parent=deform_upperarm_fr_base_l, loc=<0 0 0>, rot=<0 0 0 1>}
 deform_upperarm_up_base_l {parent=upperarm_twist_01_l, loc=<-3.2216 -1.1391 3.3311>, rot=<0 0 0 1>}
 deform_upperarm_up_l {parent=deform_upperarm_up_base_l, loc=<0 0 0>, rot=<0 0 0 1>}
 upperarm_twist_02_l {parent=upperarm_l, loc=<13.7941 0 0>, rot=<0 0 0 -1>}
 clavicle_r {parent=spine_05, loc=<4.3575 -0.1466 2.3223>, rot=<0.7289 0.0465 -0.6829 0.009>}
 upperarm_r {parent=clavicle_r, loc=<-15.5111 0.0001 0>, rot=<0.0226 -0.3234 -0.0088 -0.9459>}
 lowerarm_r {parent=upperarm_r, loc=<-26.6514 0 0>, rot=<0.0391 -0.0099 -0.2453 -0.9686>}
 hand_r {parent=lowerarm_r, loc=<-22.3491 0 0>, rot=<-0.7081 -0.0507 0.0975 -0.6975>}
 thumb_01_r {parent=hand_r, loc=<-1.3118 0.3191 2.459>, rot=<0.6546 -0.4042 0.0664 -0.6354>}
 thumb_02_r {parent=thumb_01_r, loc=<-5.9336 0.0448 -0.0401>, rot=<-0.0037 0.0213 0.298 -0.9543>}
 thumb_03_r {parent=thumb_02_r, loc=<-2.2223 -0.3682 0.0907>, rot=<0.0118 -0.013 0.0066 -0.9998>}
 index_metacarpal_r {parent=hand_r, loc=<-1.6535 -0.0183 2.96>, rot=<0.0158 -0.0124 -0.0173 -0.9996>}
 index_01_r {parent=index_metacarpal_r, loc=<-8.3008 -0.0357 0.0524>, rot=<-0.0044 0.0485 0.0924 -0.9945>}
 index_02_r {parent=index_01_r, loc=<-3.3587 0.3004 -0.0452>, rot=<-0.002 0.0027 0.3225 -0.9466>}
 index_03_r {parent=index_02_r, loc=<-2.5578 0.123 0.0252>, rot=<-0.002 0.0067 0.2558 -0.9667>}
 middle_metacarpal_r {parent=hand_r, loc=<-2.1856 0.1725 0.5984>, rot=<0.006 0.0168 -0.0329 -0.9993>}
 middle_01_r {parent=middle_metacarpal_r, loc=<-7.3275 0.2447 -0.0129>, rot=<-0.0208 0.0042 0.0858 -0.9961>}
 middle_02_r {parent=middle_01_r, loc=<-4.0123 -0.1401 -0.1175>, rot=<-0.0218 0.0258 0.3512 -0.9357>}
 middle_03_r {parent=middle_02_r, loc=<-2.6077 -0.0534 -0.1373>, rot=<0.0082 0.0188 0.1708 -0.9851>}
 ring_metacarpal_r {parent=hand_r, loc=<-2.2074 0.3625 -1.3638>, rot=<-0.0284 0.0722 -0.0458 -0.9959>}
 ring_01_r {parent=ring_metacarpal_r, loc=<-7.0295 0.6197 0.1446>, rot=<-0.0489 -0.0252 0.1892 -0.9804>}
 ring_02_r {parent=ring_01_r, loc=<-3.6174 -0.4647 0.1027>, rot=<0 0 0.2806 -0.9598>}
 ring_03_r {parent=ring_02_r, loc=<-2.4152 -0.114 -0.0199>, rot=<0 0 0.1189 -0.9929>}
 pinky_metacarpal_r {parent=hand_r, loc=<-1.7616 0.9753 -3.3814>, rot=<-0.0397 0.0814 -0.069 -0.9935>}
 pinky_01_r {parent=pinky_metacarpal_r, loc=<-7.1709 0.4831 0.1094>, rot=<-0.0654 -0.0035 0.2242 -0.9723>}
 pinky_02_r {parent=pinky_01_r, loc=<-2.8274 -0.0893 0.1043>, rot=<0.0037 0.0011 0.2854 -0.9584>}
 pinky_03_r {parent=pinky_02_r, loc=<-2.2046 -0.0307 0.0078>, rot=<-0.0002 0.0098 0.1445 -0.9895>}
 deform_wrist_dn_base_r {parent=hand_r, loc=<-0.0004 2.2345 -0.0001>, rot=<0 0 0 -1>}
 deform_wrist_dn_r {parent=deform_wrist_dn_base_r, loc=<0 0 0>, rot=<0 0 0 1>}
 deform_wrist_up_base_r {parent=hand_r, loc=<0.0003 -2.2684 0.0001>, rot=<0 0 0 -1>}
 deform_wrist_up_r {parent=deform_wrist_up_base_r, loc=<0 0 0>, rot=<0 0 0 1>}
 lowerarm_twist_02_r {parent=lowerarm_r, loc=<-9.6199 0 0>, rot=<0 0 0 -1>}
 lowerarm_twist_01_r {parent=lowerarm_r, loc=<-17.8549 0 0>, rot=<0 0 0 -1>}
 deform_elbow_in_base_r {parent=lowerarm_r, loc=<-0.2111 -5.2591 -0.8275>, rot=<0 0 0 -1>}
 deform_elbow_in_r {parent=deform_elbow_in_base_r, loc=<-0.0001 0 0.0002>, rot=<0 0 0 1>}
 deform_elbow_out_base_r {parent=lowerarm_r, loc=<0 4 -0.0001>, rot=<0 0 0 -1>}
 deform_elbow_out_r {parent=deform_elbow_out_base_r, loc=<0 0 0>, rot=<0 0 0 1>}
 upperarm_twist_01_r {parent=upperarm_r, loc=<-0.0041 0 0>, rot=<0 0 0 -1>}
 deform_upperarm_bk_base_r {parent=upperarm_twist_01_r, loc=<-1.0814 7.2192 1.9455>, rot=<0 0 0 -1>}
 deform_upperarm_bk_r {parent=deform_upperarm_bk_base_r, loc=<0 0 0>, rot=<0 0 0 1>}
 deform_upperarm_fr_base_r {parent=upperarm_twist_01_r, loc=<0.3689 -7.0507 2.7652>, rot=<0 0 0 -1>}
 deform_upperarm_fr_r {parent=deform_upperarm_fr_base_r, loc=<0 0 0>, rot=<0 0 0 1>}
 deform_upperarm_up_base_r {parent=upperarm_twist_01_r, loc=<3.2214 1.1391 -3.3308>, rot=<0 0 0 1>}
 deform_upperarm_up_r {parent=deform_upperarm_up_base_r, loc=<0 0 0>, rot=<0 0 0 -1>}
 upperarm_twist_02_r {parent=upperarm_r, loc=<-13.7941 0 0>, rot=<0 0 0 -1>}
 aim_necklace {parent=spine_05, loc=<-3.287 12.074 0>, rot=<-0.2689 -0.6617 -0.6418 -0.2792>}
 base_fur_01_lf {parent=spine_05, loc=<8.4742 9.0873 -12.3827>, rot=<0.2925 -0.5828 -0.1626 -0.7405>}
 dyn_fur_01_lf {parent=base_fur_01_lf, loc=<0 0 0>, rot=<0 0 0 1>}
 base_fur_02_lf {parent=spine_05, loc=<11.7609 3.7582 -9.2854>, rot=<0.4199 -0.5612 -0.0131 -0.7131>}
 dyn_fur_02_lf {parent=base_fur_02_lf, loc=<0 0 0>, rot=<0 0 0 1>}
 base_fur_03_lf {parent=spine_05, loc=<15.6373 -1.8657 -5.7238>, rot=<0.4663 -0.5366 0.0457 -0.7019>}
 dyn_fur_03_lf {parent=base_fur_03_lf, loc=<0 0 0>, rot=<0 0 0 1>}
 base_fur_01_cr {parent=spine_05, loc=<17.2913 -3.7398 0>, rot=<0.6111 -0.3557 0.3557 -0.6111>}
 dyn_fur_01_cr {parent=base_fur_01_cr, loc=<0 0 0>, rot=<0 0 0 1>}
 base_fur_03_rt {parent=spine_05, loc=<15.6368 -1.8657 5.7238>, rot=<0.7019 -0.0457 0.5366 -0.4663>}
 dyn_fur_03_rt {parent=base_fur_03_rt, loc=<0 0 0>, rot=<0 0 0 1>}
 base_fur_02_rt {parent=spine_05, loc=<11.7607 3.7582 9.2854>, rot=<0.7131 0.0131 0.5612 -0.4199>}
 dyn_fur_02_rt {parent=base_fur_02_rt, loc=<0 0 0>, rot=<0 0 0 1>}
 base_fur_01_rt {parent=spine_05, loc=<8.474 9.0873 12.3827>, rot=<0.7405 0.1626 0.5828 -0.2925>}
 dyn_fur_01_rt {parent=base_fur_01_rt, loc=<0 0 0>, rot=<0 0 0 1>}
 base_necklace {parent=spine_05, loc=<9.3683 5.0422 -0.0152>, rot=<-0.2689 -0.6617 -0.6418 -0.2792>}
 dyn_necklace_01 {parent=base_necklace, loc=<0 0 0>, rot=<0 0 0 1>}
 dyn_necklace_02 {parent=dyn_necklace_01, loc=<6.5642 0 0>, rot=<-0.0006 -0.0452 0.0138 -0.9989>}
 dyn_necklace_03 {parent=dyn_necklace_02, loc=<4.163 0 0>, rot=<0.0766 -0.2158 0.017 -0.9733>}
 dyn_dogtags_01 {parent=dyn_necklace_03, loc=<1.7025 -0.2517 0.2282>, rot=<-0.0451 0.022 -0.05 -0.9975>}
 dyn_dogtags_02 {parent=dyn_necklace_03, loc=<1.5671 0.0653 -0.2549>, rot=<-0.0547 -0.0381 0.1212 -0.9904>}
 neck_01_get_rot_offset {parent=spine_05, loc=<11.3138 0.1529 0>, rot=<0 0 -0.2003 -0.9797>}
 neck_01_get_rot {parent=neck_01_get_rot_offset, loc=<0 0 0>, rot=<0 0 0 1>}
 neck_01_set_rot_offset {parent=spine_05, loc=<11.3138 0.1529 0>, rot=<0 0 -0.2003 -0.9797>}
 neck_01_set_rot {parent=neck_01_set_rot_offset, loc=<0 0 0>, rot=<0 0 0 1>}
 clavicle_l_null {parent=spine_05, type=Null, loc=<4.3575 -0.1466 -2.3223>, rot=<0.0465 -0.7289 -0.009 -0.6829>}
 clavicle_r_null {parent=spine_05, type=Null, loc=<4.3575 -0.1466 2.3223>, rot=<0.7289 0.0465 -0.6829 0.009>}
 deform_pec_aim_base_l {parent=spine_04, loc=<7.8732 12.6388 -2.4123>, rot=<0.0102 -0.4602 0.3979 -0.7936>}
 deform_pec_aim_l {parent=deform_pec_aim_base_l, loc=<0 0 0>, rot=<0 0 0 1>}
 deform_pec_base_l {parent=deform_pec_aim_l, loc=<10.766 0 0>, rot=<0 0 0 1>}
 deform_pec_l {parent=deform_pec_base_l, loc=<0.055 0 0>, rot=<0 0 0 1>}
 deform_pec_aim_base_r {parent=spine_04, loc=<7.8727 12.6388 2.4123>, rot=<-0.4602 -0.0102 0.7936 0.3979>}
 deform_pec_aim_r {parent=deform_pec_aim_base_r, loc=<0 0 0>, rot=<0 0 0 1>}
 deform_pec_base_r {parent=deform_pec_aim_r, loc=<-10.7662 -0.0006 -0.0007>, rot=<0 0 0 1>}
 deform_pec_r {parent=deform_pec_base_r, loc=<0 0 0>, rot=<0 0 0 1>}
 base_main_jacket_01_cr {parent=spine_03, loc=<0 0 0>, rot=<0 0 0 1>}
 offset_main_jacket_01_cr {parent=base_main_jacket_01_cr, loc=<0 0 0>, rot=<0 0 0 1>}
 dyn_main_jacket_01_cr {parent=offset_main_jacket_01_cr, loc=<0 0 0>, rot=<0 0 0 1>}
 base_jacket_01_cr {parent=dyn_main_jacket_01_cr, loc=<5.9159 -10.4302 0>, rot=<0.0801 -0.9968 0 0>}
 dyn_jacket_01_cr {parent=base_jacket_01_cr, loc=<0 0 0>, rot=<0 0 0 1>}
 dyn_jacket_01_cr_end {parent=dyn_jacket_01_cr, loc=<5.97 0 0>, rot=<0 0 0 1>}
 base_jacket_03_lf {parent=dyn_main_jacket_01_cr, loc=<5.3296 -6.8403 -12.3275>, rot=<0.0674 -0.9164 -0.3942 0.0159>}
 dyn_jacket_03_lf {parent=base_jacket_03_lf, loc=<0 0 0>, rot=<0 0 0 1>}
 dyn_jacket_03_lf_end {parent=dyn_jacket_03_lf, loc=<5.97 0 0>, rot=<0 0 0 1>}
 base_jacket_02_lf {parent=dyn_main_jacket_01_cr, loc=<3.8393 2.2848 -16.2823>, rot=<-0.0437 0.6943 0.7169 -0.046>}
 dyn_jacket_02_lf {parent=base_jacket_02_lf, loc=<0 0 0>, rot=<0 0 0 1>}
 dyn_jacket_02_lf_end {parent=dyn_jacket_02_lf, loc=<5.97 0 0>, rot=<0 0 0 1>}
 base_jacket_01_lf {parent=dyn_main_jacket_01_cr, loc=<2.5988 9.8811 -12.3275>, rot=<-0.0238 0.4829 0.8731 -0.0624>}
 dyn_jacket_01_lf {parent=base_jacket_01_lf, loc=<0 0 0>, rot=<0 0 0 1>}
 dyn_jacket_01_lf_end {parent=dyn_jacket_01_lf, loc=<5.97 0 0>, rot=<0 0 0 1>}
 base_jacket_03_rt {parent=dyn_main_jacket_01_cr, loc=<5.3294 -6.8403 12.3275>, rot=<0.0674 -0.9164 0.3942 -0.0159>}
 dyn_jacket_03_rt {parent=base_jacket_03_rt, loc=<0 0 0>, rot=<0 0 0 1>}
 dyn_jacket_03_rt_end {parent=dyn_jacket_03_rt, loc=<5.97 0 0>, rot=<0 0 0 1>}
 base_jacket_02_rt {parent=dyn_main_jacket_01_cr, loc=<3.839 2.2848 16.2823>, rot=<0.0437 -0.6943 0.7169 -0.046>}
 dyn_jacket_02_rt {parent=base_jacket_02_rt, loc=<0 0 0>, rot=<0 0 0 1>}
 dyn_jacket_02_rt_end {parent=dyn_jacket_02_rt, loc=<5.97 0 0>, rot=<0 0 0 1>}
 base_jacket_01_rt {parent=dyn_main_jacket_01_cr, loc=<2.5983 9.8811 12.3275>, rot=<0.0238 -0.4829 0.8731 -0.0624>}
 dyn_jacket_01_rt {parent=base_jacket_01_rt, loc=<0 0 0>, rot=<0 0 0 1>}
 dyn_jacket_01_rt_end {parent=dyn_jacket_01_rt, loc=<5.97 0 0>, rot=<0 0 0 1>}
 base_zipper {parent=dyn_jacket_01_rt, loc=<13.1316 -1.3336 -2.8749>, rot=<0.0497 -0.0559 0.1443 -0.9867>}
 dyn_zipper {parent=base_zipper, loc=<0 0 0>, rot=<0 0 0 1>}
 base_jacket_helper {parent=dyn_main_jacket_01_cr, loc=<5.1645 1.43 -17.4606>, rot=<0.0387 -0.706 -0.0387 -0.706>}
 dyn_jacket_helper {parent=base_jacket_helper, loc=<0 0 0>, rot=<0 0 0 1>}
 base_jacket_helper_lf {parent=dyn_main_jacket_01_cr, loc=<5.1645 1.43 -17.4606>, rot=<0.0387 -0.706 -0.0387 -0.706>}
 dyn_jacket_helper_lf {parent=base_jacket_helper_lf, loc=<0 0 0>, rot=<0 0 0 1>}
 base_jacket_helper_rt {parent=dyn_main_jacket_01_cr, loc=<5.1644 1.43 17.4606>, rot=<0.706 0.0387 0.706 -0.0387>}
 dyn_jacket_helper_rt {parent=base_jacket_helper_rt, loc=<0 0 0>, rot=<0 0 0 1>}
 spine_03_for_cons_offset {parent=spine_02, loc=<6.0104 0 0>, rot=<0 0 0.0149 -0.9999>}
 spine_03_for_cons {parent=spine_03_for_cons_offset, loc=<0 0 0>, rot=<0 0 0 1>}
 spine_02_for_cons_offset {parent=spine_01, loc=<7.2518 -0.035 0>, rot=<0 0 0.0238 -0.9997>}
 spine_02_for_cons {parent=spine_02_for_cons_offset, loc=<0 0 0>, rot=<0 0 0 1>}
 thigh_l {parent=pelvis, loc=<-2.5693 -0.0109 -7.7501>, rot=<0.0012 0.0572 0.0155 -0.9982>}
 calf_l {parent=thigh_l, loc=<-39.0484 0 0>, rot=<0.0006 -0.0161 -0.0391 -0.9991>}
 foot_l {parent=calf_l, loc=<-42.561 0.1077 0.4292>, rot=<0.0418 -0.0455 0.0343 -0.9975>}
 ball_l {parent=foot_l, loc=<-6.1599 13.8264 0.6806>, rot=<0.0001 -0.0001 -0.7063 -0.7079>}
 calf_twist_02_l {parent=calf_l, loc=<-18.8064 -0.0967 0.0397>, rot=<0.0237 0.0095 -0.0035 -0.9997>}
 calf_twist_01_l {parent=calf_l, loc=<-35.813 -0.0967 0.0397>, rot=<0.0237 0.0095 -0.0035 -0.9997>}
 deform_knee_in_base_l {parent=calf_l, loc=<1.298 -5.2957 -0.0496>, rot=<0 0 0 -1>}
 deform_knee_in_l {parent=deform_knee_in_base_l, loc=<0 0 0>, rot=<0 0 0 1>}
 deform_knee_out_base_l {parent=calf_l, loc=<0 4 0>, rot=<0 0 0 -1>}
 deform_knee_out_l {parent=deform_knee_out_base_l, loc=<0 0 0>, rot=<0 0 0 1>}
 thigh_twist_01_l {parent=thigh_l, loc=<-21.6303 0 0>, rot=<0.0013 0 0 -1>}
 deform_groin_base_l {parent=thigh_l, loc=<-2.1908 8.1142 -1.7442>, rot=<0 0 0 -1>}
 deform_groin_l {parent=deform_groin_base_l, loc=<0 0 0>, rot=<0 0 0 1>}
 thigh_r {parent=pelvis, loc=<-2.5693 -0.0109 7.7501>, rot=<0.0572 -0.0012 0.9982 0.0155>}
 calf_r {parent=thigh_r, loc=<39.0484 0 0>, rot=<0.0006 -0.0161 -0.0391 -0.9991>}
 foot_r {parent=calf_r, loc=<42.561 -0.1077 -0.4292>, rot=<0.0418 -0.0455 0.0343 -0.9975>}
 ball_r {parent=foot_r, loc=<6.1599 -13.8264 -0.6806>, rot=<0.0001 -0.0001 -0.7063 -0.7079>}
 calf_twist_02_r {parent=calf_r, loc=<18.8064 0.0967 -0.0397>, rot=<0.0237 0.0095 -0.0035 -0.9997>}
 calf_twist_01_r {parent=calf_r, loc=<35.813 0.0967 -0.0397>, rot=<0.0237 0.0095 -0.0035 -0.9997>}
 deform_knee_in_base_r {parent=calf_r, loc=<-1.298 5.2957 0.0496>, rot=<0 0 0 -1>}
 deform_knee_in_r {parent=deform_knee_in_base_r, loc=<0 0 0>, rot=<0 0 0 1>}
 deform_knee_out_base_r {parent=calf_r, loc=<0 -4 0>, rot=<0 0 0 -1>}
 deform_knee_out_r {parent=deform_knee_out_base_r, loc=<0 0 0>, rot=<0 0 0 1>}
 thigh_twist_01_r {parent=thigh_r, loc=<21.6303 0 0>, rot=<0.0013 0 0 -1>}
 deform_groin_base_r {parent=thigh_r, loc=<2.1908 -8.1142 1.7442>, rot=<0 0 0 -1>}
 deform_groin_r {parent=deform_groin_base_r, loc=<0 0 0>, rot=<0 0 0 1>}
 deform_glute_base_l {parent=pelvis, loc=<-8.8064 -8.8222 -8.4287>, rot=<0.0275 0.0501 -0.4472 -0.8926>}
 deform_glute_l {parent=deform_glute_base_l, loc=<0 0 0>, rot=<0 0 0 1>}
 deform_glute_base_r {parent=pelvis, loc=<-8.8062 -8.8222 8.4287>, rot=<0.0501 -0.0275 0.8926 -0.4472>}
 deform_glute_r {parent=deform_glute_base_r, loc=<0 0 0>, rot=<0 0 0 1>}
 base_gun_rt {parent=pelvis, loc=<6.39 0.5748 13.9598>, rot=<0.0549 -0.7684 0.6339 0.0682>}
 dyn_gun_rt {parent=base_gun_rt, loc=<0 0 0>, rot=<0 0 0 1>}
 base_bullet_01_rt {parent=pelvis, loc=<7.3986 -11.7296 8.6306>, rot=<0.0591 -0.9706 0.1997 0.1203>}
 dyn_bullet_01_rt {parent=base_bullet_01_rt, loc=<0 0 0>, rot=<0 0 0 1>}
 base_bullet_02_rt {parent=pelvis, loc=<7.9214 -10.3509 10.7762>, rot=<0.0444 -0.9526 0.2843 0.0987>}
 dyn_bullet_02_rt {parent=base_bullet_02_rt, loc=<0 0 0>, rot=<0 0 0 1>}
 base_bullet_03_rt {parent=pelvis, loc=<8.1955 -8.5328 12.5486>, rot=<0.0377 -0.9263 0.3643 0.0888>}
 dyn_bullet_03_rt {parent=base_bullet_03_rt, loc=<0 0 0>, rot=<0 0 0 1>}
 base_strap_back_01_lf {parent=pelvis, loc=<9.1386 -9.5782 -6.4524>, rot=<0.1597 -0.7739 0.5778 -0.2041>}
 dyn_strap_back_01_lf {parent=base_strap_back_01_lf, loc=<0 0 0>, rot=<0 0 0 1>}
 dyn_strap_back_02_lf {parent=dyn_strap_back_01_lf, loc=<4.9801 0 0>, rot=<0 -0.2036 0 -0.9791>}
 dyn_strap_back_03_lf {parent=dyn_strap_back_02_lf, loc=<6.1904 0 0>, rot=<0.0008 -0.0672 -0.0116 -0.9977>}
 dyn_strap_back_04_lf {parent=dyn_strap_back_03_lf, loc=<7.3554 0 0>, rot=<0 -0.0042 -0.0007 -1>}
 dyn_strap_back_05_lf {parent=dyn_strap_back_04_lf, loc=<3.8355 0 0>, rot=<0 0 0 1>}
 base_chain_front_01_lf {parent=pelvis, loc=<1.7194 4.2054 -12.7097>, rot=<-0.0322 -0.9122 -0.1814 -0.366>}
 dyn_chain_front_01_lf {parent=base_chain_front_01_lf, loc=<0 0 0>, rot=<0 0 0 1>}
 dyn_chain_front_02_lf {parent=dyn_chain_front_01_lf, loc=<0.9313 0 0>, rot=<0.0036 0.0614 -0.0579 -0.9964>}
 dyn_chain_front_03_lf {parent=dyn_chain_front_02_lf, loc=<1.3061 0 0>, rot=<0.0007 -0.064 -0.0113 -0.9979>}
 dyn_chain_front_04_lf {parent=dyn_chain_front_03_lf, loc=<1.4058 0 0>, rot=<0.0005 0.0448 0.0121 -0.9989>}
 dyn_chain_front_05_lf {parent=dyn_chain_front_04_lf, loc=<1.3521 0 0>, rot=<0 0 0 1>}
 thigh_l_get_rot_offset {parent=pelvis, loc=<-2.5693 -0.0109 -7.7501>, rot=<0.0012 0.0572 0.0155 -0.9982>}
 thigh_l_get_rot {parent=thigh_l_get_rot_offset, loc=<0 0 0>, rot=<0 0 0 1>}
 thigh_l_set_rot_offset {parent=pelvis, loc=<-2.5693 -0.0109 -7.7501>, rot=<0.0012 0.0572 0.0155 -0.9982>}
 thigh_l_set_rot {parent=thigh_l_set_rot_offset, loc=<0 0 0>, rot=<0 0 0 1>}
 thigh_r_get_rot_offset {parent=pelvis, loc=<-2.5693 -0.0109 7.7501>, rot=<0.0572 -0.0012 0.9982 0.0155>}
 thigh_r_get_rot {parent=thigh_r_get_rot_offset, loc=<0 0 0>, rot=<0 0 0 1>}
 thigh_r_set_rot_offset {parent=pelvis, loc=<-2.5693 -0.0109 7.7501>, rot=<0.0572 -0.0012 0.9982 0.0155>}
 thigh_r_set_rot {parent=thigh_r_set_rot_offset, loc=<0 0 0>, rot=<0 0 0 1>}
 crouching alpha {type=Curve, loc=<0 0 0>, rot=<0 0 0 1>}
]
variables [
 isCrouching {type=bool, default=False}
 isCrouchingCR {type=double, default=0}
 isCrouchingMoveCR {type=double, default=0}
]
