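# Physics asset for the FunkyBear male medium cosmetic jacket, providing capsule bodies and constraints for its dynamic bones.
# A main jacket body links to the pelvis and fans out through free constraints to seven small base anchors, each leading a two-capsule chain.
# Chain joints allow limited swing on one axis with zero-degree twist and second swing, tighter on the first left chain.
# An unconstrained left thigh capsule has no right-side counterpart.
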PhysicsAsset /BRCosmetics/Characters/Player/Male/Medium/Bodies/M_MED_FunkyBear/Meshes/M_MED_FunkyBear_Jacket_Physics
bodies {
 base_jacket_01_cr [{shapeType=Capsule, length=0.505, radius=0.505, shapeName=base_jacket_01_cr_capsule, rotation=<rot 0 -90 0>}]
 base_jacket_01_lf [{shapeType=Capsule, length=0.505, radius=0.505, shapeName=base_jacket_01_lf_capsule, rotation=<rot 0 -90 0>}]
 base_jacket_01_rt [{shapeType=Capsule, length=0.505, radius=0.505, shapeName=base_jacket_01_rt_capsule, rotation=<rot 0 -90 0>}]
 base_jacket_02_lf [{shapeType=Capsule, length=0.505, radius=0.505, shapeName=base_jacket_02_lf_capsule, rotation=<rot 0 -90 0>}]
 base_jacket_02_rt [{shapeType=Capsule, length=0.505, radius=0.505, shapeName=base_jacket_02_rt_capsule, rotation=<rot 0 -90 0>}]
 base_jacket_03_lf [{shapeType=Capsule, length=0.505, radius=0.505, shapeName=base_jacket_03_lf_capsule, rotation=<rot 0 -90 0>}]
 base_jacket_03_rt [{shapeType=Capsule, length=0.505, radius=0.505, shapeName=base_jacket_03_rt_capsule, rotation=<rot 0 -90 0>}]
 dyn_jacket_01_cr [{shapeType=Capsule, length=8, radius=4, shapeName=dyn_jacket_01_cr_capsule, center=<2 0 0>}]
 dyn_jacket_01_cr_end [{shapeType=Capsule, length=6, radius=3, shapeName=dyn_jacket_01_cr_end_capsule, center=<5 0 0>}]
 dyn_jacket_01_lf [{shapeType=Capsule, length=8, radius=4, shapeName=dyn_jacket_01_lf_capsule, center=<2 0 0>}]
 dyn_jacket_01_lf_end [{shapeType=Capsule, length=6, radius=3, shapeName=dyn_jacket_01_lf_end_capsule, center=<5 0 0>, rotation=<rot -3.2484 0 0>}]
 dyn_jacket_01_rt [{shapeType=Capsule, length=8, radius=4, shapeName=dyn_jacket_01_rt_capsule, center=<2 0 0>}]
 dyn_jacket_01_rt_end [{shapeType=Capsule, length=6, radius=3, shapeName=dyn_jacket_01_rt_end_capsule, center=<5 0 0>}]
 dyn_jacket_02_lf [{shapeType=Capsule, length=8, radius=4, shapeName=dyn_jacket_02_lf_capsule, center=<2 0 0>}]
 dyn_jacket_02_lf_end [{shapeType=Capsule, length=6, radius=3, shapeName=dyn_jacket_02_lf_end_capsule, center=<5 0 0>}]
 dyn_jacket_02_rt [{shapeType=Capsule, length=8, radius=4, shapeName=dyn_jacket_02_rt_capsule, center=<2 0 0>}]
 dyn_jacket_02_rt_end [{shapeType=Capsule, length=6, radius=3, shapeName=dyn_jacket_02_rt_end_capsule, center=<5 0 0>}]
 dyn_jacket_03_lf [{shapeType=Capsule, length=8, radius=4, shapeName=dyn_jacket_03_lf_capsule, center=<2 0 0>}]
 dyn_jacket_03_lf_end [{shapeType=Capsule, length=6, radius=3, shapeName=dyn_jacket_03_lf_end_capsule, center=<5 0 0>}]
 dyn_jacket_03_rt [{shapeType=Capsule, length=8, radius=4, shapeName=dyn_jacket_03_rt_capsule, center=<2 0 0>}]
 dyn_jacket_03_rt_end [{shapeType=Capsule, length=6, radius=3, shapeName=dyn_jacket_03_rt_end_capsule, center=<5 0 0>}]
 dyn_main_jacket_01_cr [{shapeType=Capsule, length=0.505, radius=0.505, shapeName=dyn_main_jacket_01_cr_capsule, rotation=<rot 0 -90 0>}]
 pelvis [{shapeType=Capsule, length=4.4698, radius=11.5798, shapeName=pelvis_capsule, center=<-1.1436 -0.9852 0>}]
 thigh_l [{shapeType=Capsule, length=48.2264, radius=8.483, shapeName=thigh_l_capsule, center=<-24.3385 -0.0947 0>, rotation=<rot 90 -4.2253 0>}]
}
constraints [
 {
  bone1Name=dyn_main_jacket_01_cr
  bone2Name=pelvis
  swing1LimitDegrees=45
  swing1Motion=Free
  swing2LimitDegrees=45
  swing2Motion=Free
  twistLimitDegrees=45
  twistMotion=Free
 }
 {
  bone1Name=base_jacket_01_cr
  bone2Name=dyn_main_jacket_01_cr
  swing1LimitDegrees=45
  swing1Motion=Free
  swing2LimitDegrees=45
  swing2Motion=Free
  twistLimitDegrees=45
  twistMotion=Free
 }
 {bone1Name=dyn_jacket_01_cr, bone2Name=base_jacket_01_cr, swing1LimitDegrees=20, swing2LimitDegrees=0, twistLimitDegrees=0}
 {bone1Name=dyn_jacket_01_cr_end, bone2Name=dyn_jacket_01_cr, swing1LimitDegrees=20, swing2LimitDegrees=0, twistLimitDegrees=0}
 {
  bone1Name=base_jacket_01_lf
  bone2Name=dyn_main_jacket_01_cr
  swing1LimitDegrees=45
  swing1Motion=Free
  swing2LimitDegrees=45
  swing2Motion=Free
  twistLimitDegrees=45
  twistMotion=Free
 }
 {bone1Name=dyn_jacket_01_lf, bone2Name=base_jacket_01_lf, swing1LimitDegrees=5, swing2LimitDegrees=0, twistLimitDegrees=0}
 {bone1Name=dyn_jacket_01_lf_end, bone2Name=dyn_jacket_01_lf, swing1LimitDegrees=5, swing2LimitDegrees=0, twistLimitDegrees=0}
 {
  bone1Name=base_jacket_01_rt
  bone2Name=dyn_main_jacket_01_cr
  swing1LimitDegrees=45
  swing1Motion=Free
  swing2LimitDegrees=45
  swing2Motion=Free
  twistLimitDegrees=45
  twistMotion=Free
 }
 {bone1Name=dyn_jacket_01_rt, bone2Name=base_jacket_01_rt, swing1LimitDegrees=20, swing2LimitDegrees=0, twistLimitDegrees=0}
 {bone1Name=dyn_jacket_01_rt_end, bone2Name=dyn_jacket_01_rt, swing1LimitDegrees=20, swing2LimitDegrees=0, twistLimitDegrees=0}
 {
  bone1Name=base_jacket_02_lf
  bone2Name=dyn_main_jacket_01_cr
  swing1LimitDegrees=45
  swing1Motion=Free
  swing2LimitDegrees=45
  swing2Motion=Free
  twistLimitDegrees=45
  twistMotion=Free
 }
 {bone1Name=dyn_jacket_02_lf, bone2Name=base_jacket_02_lf, swing1LimitDegrees=20, swing2LimitDegrees=0, twistLimitDegrees=0}
 {bone1Name=dyn_jacket_02_lf_end, bone2Name=dyn_jacket_02_lf, swing1LimitDegrees=20, swing2LimitDegrees=0, twistLimitDegrees=0}
 {
  bone1Name=base_jacket_02_rt
  bone2Name=dyn_main_jacket_01_cr
  swing1LimitDegrees=45
  swing1Motion=Free
  swing2LimitDegrees=45
  swing2Motion=Free
  twistLimitDegrees=45
  twistMotion=Free
 }
 {bone1Name=dyn_jacket_02_rt, bone2Name=base_jacket_02_rt, swing1LimitDegrees=20, swing2LimitDegrees=0, twistLimitDegrees=0}
 {bone1Name=dyn_jacket_02_rt_end, bone2Name=dyn_jacket_02_rt, swing1LimitDegrees=20, swing2LimitDegrees=0, twistLimitDegrees=0}
 {
  bone1Name=base_jacket_03_lf
  bone2Name=dyn_main_jacket_01_cr
  swing1LimitDegrees=45
  swing1Motion=Free
  swing2LimitDegrees=45
  swing2Motion=Free
  twistLimitDegrees=45
  twistMotion=Free
 }
 {bone1Name=dyn_jacket_03_lf, bone2Name=base_jacket_03_lf, swing1LimitDegrees=20, swing2LimitDegrees=0, twistLimitDegrees=0}
 {bone1Name=dyn_jacket_03_lf_end, bone2Name=dyn_jacket_03_lf, swing1LimitDegrees=20, swing2LimitDegrees=0, twistLimitDegrees=0}
 {
  bone1Name=base_jacket_03_rt
  bone2Name=dyn_main_jacket_01_cr
  swing1LimitDegrees=45
  swing1Motion=Free
  swing2LimitDegrees=45
  swing2Motion=Free
  twistLimitDegrees=45
  twistMotion=Free
 }
 {bone1Name=dyn_jacket_03_rt, bone2Name=base_jacket_03_rt, swing1LimitDegrees=20, swing2LimitDegrees=0, twistLimitDegrees=0}
 {bone1Name=dyn_jacket_03_rt_end, bone2Name=dyn_jacket_03_rt, swing1LimitDegrees=20, swing2LimitDegrees=0, twistLimitDegrees=0}
]
