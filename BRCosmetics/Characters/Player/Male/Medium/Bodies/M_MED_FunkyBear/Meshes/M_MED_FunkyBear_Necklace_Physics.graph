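# Physics asset for the necklace and dog tags on the male medium FunkyBear player body cosmetic.
# Constraints chain three necklace capsules from a small base body, then branch to two dog tag capsules at the last link.
# Each joint has narrow swing limits and a zero twist limit.
# Two large overlapping capsules on spine_05 form an unconstrained upper-torso body, likely for the necklace to collide against.
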PhysicsAsset /BRCosmetics/Characters/Player/Male/Medium/Bodies/M_MED_FunkyBear/Meshes/M_MED_FunkyBear_Necklace_Physics
bodies {
 base_necklace [{shapeType=Capsule, length=0.505, radius=0.505, shapeName=base_necklace_capsule, rotation=<rot 0 -90 0>}]
 dyn_dogtags_01 [{shapeType=Capsule, length=3.9697, radius=1.7507, shapeName=dyn_dogtags_01_capsule, center=<2 0 0>, rotation=<rot 90 0 0>}]
 dyn_dogtags_02 [{shapeType=Capsule, length=4.1932, radius=1.7527, shapeName=dyn_dogtags_02_capsule, center=<2 0 0>, rotation=<rot 90 0 0>}]
 dyn_necklace_01 [{shapeType=Capsule, length=10.5755, radius=1.5302, shapeName=dyn_necklace_01_capsule, center=<2 0 2>, rotation=<rot 0.0002 0 91.2334>}]
 dyn_necklace_02 [{shapeType=Capsule, length=10.5486, radius=1.5571, shapeName=dyn_necklace_02_capsule, center=<2 0 0.5>, rotation=<rot 0.0002 0 93.0983>}]
 dyn_necklace_03 [{shapeType=Capsule, length=10.3743, radius=1.7314, shapeName=dyn_necklace_03_capsule, center=<2 0 0>, rotation=<rot 0.0002 0 82.9765>}]
 spine_05 [
  {shapeType=Capsule, length=16.8205, radius=12.5847, shapeName=spine_05_capsule, center=<-4.053 -1.3498 0>}
  {shapeType=Capsule, length=21.5466, radius=11.6538, shapeName=spine_05_capsule, center=<-0.1156 -1.3154 0>}
 ]
}
constraints [
 {bone1Name=dyn_necklace_01, bone2Name=base_necklace, swing1LimitDegrees=15, swing2LimitDegrees=15, twistLimitDegrees=0}
 {bone1Name=dyn_necklace_02, bone2Name=dyn_necklace_01, swing1LimitDegrees=20, swing2LimitDegrees=15, twistLimitDegrees=0}
 {bone1Name=dyn_necklace_03, bone2Name=dyn_necklace_02, swing1LimitDegrees=20, swing2LimitDegrees=15, twistLimitDegrees=0}
 {bone1Name=dyn_dogtags_01, bone2Name=dyn_necklace_03, swing1LimitDegrees=20, swing2LimitDegrees=10, twistLimitDegrees=0}
 {bone1Name=dyn_dogtags_02, bone2Name=dyn_necklace_03, swing1LimitDegrees=20, swing2LimitDegrees=5, twistLimitDegrees=0}
]
