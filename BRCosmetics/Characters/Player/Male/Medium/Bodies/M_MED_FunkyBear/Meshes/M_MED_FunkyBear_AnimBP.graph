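# Animation Blueprint for the FunkyBear character body that follows the attached parent mesh and adds secondary motion to its accessories.
# The copied pose runs through a crouch-aware Control Rig, head adjustment, RigidBody jacket, strap-chain and necklace sims, then AnimDynamics dog tags and zipper.
# A front-chain offset follows, with front-end idle curves blending slot animation onto chain and gun bones.
# The EventGraph computes gravity override, scaled counter head rotation, and skydive, crouch and zipline tuning values.
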AnimBlueprint /BRCosmetics/Characters/Player/Male/Medium/Bodies/M_MED_FunkyBear/Meshes/M_MED_FunkyBear_AnimBP
target_skeleton=&/M_MED_FunkyBear_Skeleton
graphs [
 AnimGraph {nodes [
   n1 = Layered blend per bone(
    BasePose=Use cached pose 'postFixModify'()
    BlendPoses_0=Slot 'DefaultSlot'()
    BlendWeights_0=Get Custom Idle Is Playing Float() -> (Custom Idle Is Playing Float:double)
   ) {layerSetup [{branchFilters [
       {boneName=base_chain_front_01_lf}
       {boneName=dyn_chain_front_01_lf}
       {boneName=dyn_chain_front_02_lf}
       {boneName=dyn_chain_front_03_lf}
       {boneName=dyn_chain_front_04_lf}
       {boneName=dyn_chain_front_05_lf}
       {boneName=base_gun_rt}
       {boneName=dyn_gun_rt}
      ]}]}
   Output Pose(Result=@n1) {layerGroup=None}
   n2 = Control Rig!(Source=Copy Pose From Mesh() {bUseAttachedParent=true}, Alpha, isCrouching:bool=Get bIsCrouching()) -> (Pose) {controlRigAssetReference {blueprintRigClass=&/M_MED_FunkyBear_CtrlRig.M_MED_FunkyBear_CtrlRig_C}}
   Save cached pose 'controlRigCache'(Pose=@n2) {cacheName=controlRigCache}
   n3 = "Transform (Modify) Bone - Bone: head"!(ComponentPose=Local To Component(LocalPose=Use cached pose 'controlRigCache'()), Alpha=1, Rotation="0,0,0") -> (Pose) {rotationMode=BMM_Additive, rotationSpace=BCS_BoneSpace, alphaBoolBlend {blendOption=Linear}, boneToModify {boneName=head}}
   Save cached pose 'modifyTransPreCache'(Pose=Component To Local(ComponentPose=@n3)) {cacheName=modifyTransPreCache}
   n4 = "Transform (Modify) Bone - Bone: base_chain_front_01_lf"(
    ComponentPose=Local To Component(LocalPose=Use cached pose 'dynamicsCache'())
    Alpha=1
    Translation="0,0,0"
    Rotation="-10,2,0"
    Scale="1,1,1"
   ) {
    rotationMode=BMM_Additive
    rotationSpace=BCS_BoneSpace
    translationMode=BMM_Additive
    translationSpace=BCS_BoneSpace
    boneToModify {boneName=base_chain_front_01_lf}
    rotation {pitch=-10, yaw=2}
   }
   Save cached pose 'postFixModify'(Pose=Component To Local(ComponentPose=@n4)) {cacheName=postFixModify}
   n5 = RigidBody!(ComponentPose=Local To Component(LocalPose=Use cached pose 'modifyTransPreCache'()), Alpha, OverrideWorldGravity:Vector="0,0,0") -> (Pose) {
    bClampLinearTranslationLimitToRefPose=true
    bOverrideWorldGravity=true
    overridePhysicsAsset=&/M_MED_FunkyBear_Jacket_Physics
    simulationSpace=BaseBoneSpace
    alphaBoolBlend {blendOption=Linear}
    alphaScaleBiasClamp {bInterpResult=true}
    baseBoneRef {boneName=spine_01}
    simSpaceSettings {worldAlpha=0.3}
   }
   n6 = RigidBody(ComponentPose=@n5, OverrideWorldGravity="0,0,0") {
    bClampLinearTranslationLimitToRefPose=true
    bOverrideWorldGravity=true
    overridePhysicsAsset=&/M_MED_FunkyBear_BackStrap_Chain_Physics
    simulationSpace=BaseBoneSpace
    baseBoneRef {boneName=pelvis}
    simSpaceSettings {
     dampingAlpha=0.8
     maxAngularAcceleration=6000
     maxAngularVelocity=6000
     maxLinearAcceleration=6000
     maxLinearVelocity=6000
     velocityScaleZ=0.9
     worldAlpha=0.5
    }
   }
   n7 = RigidBody(ComponentPose=@n6, OverrideWorldGravity="0,0,0") {
    bClampLinearTranslationLimitToRefPose=true
    bOverrideWorldGravity=true
    overridePhysicsAsset=&/M_MED_FunkyBear_Necklace_Physics
    simulationSpace=BaseBoneSpace
    baseBoneRef {boneName=spine_05}
    simSpaceSettings {
     dampingAlpha=0.5
     maxAngularAcceleration=6000
     maxAngularVelocity=6000
     maxLinearAcceleration=6000
     maxLinearVelocity=6000
     velocityScaleZ=0.85
     worldAlpha=0.5
    }
   }
   n8 = Anim Dynamics - Bone: dyn_dogtags_01!(ComponentPose=@n7, Alpha=1, ComponentLinearAccScale:Vector="0.5,0.5,0.5", ComponentLinearVelScale:Vector="0.5,0.5,0.5") -> (Pose) {
    angularDampingOverride=1
    angularSpringConstant=20000
    bAngularSpring=true
    bOverrideAngularDamping=true
    bUseGravityOverride=true
    gravityScale=0.35
    simulationSpace=RootRelative
    boundBone {boneName=dyn_dogtags_01}
    componentAppliedLinearAccClamp=<3000 3000 3000>
    componentLinearAccScale=<0.5 0.5 0.5>
    componentLinearVelScale=<0.5 0.5 0.5>
    gravityOverride {z=-980}
    physicsBodyDefinitions [{
      boundBone {boneName=dyn_dogtags_01}
      boxExtents=<30 30 30>
      constraintSetup {angularLimitsMax=<0 0 30>, angularLimitsMin=<0 -30 0>, angularTarget=<1 0 0>}
      localJointOffset=<8 0 0>
     }]
    simSpaceSettings {maxAngularAcceleration=35, maxAngularVelocity=7, simSpaceAngularAlpha=0.5}
   }
   n9 = Anim Dynamics - Bone: dyn_dogtags_02!(ComponentPose=@n8, Alpha=1, ComponentLinearAccScale:Vector="0.5,0.5,0.5", ComponentLinearVelScale:Vector="0.5,0.5,0.5") -> (Pose) {
    angularDampingOverride=1
    angularSpringConstant=20000
    bAngularSpring=true
    bOverrideAngularDamping=true
    bUseGravityOverride=true
    gravityScale=0.35
    simulationSpace=RootRelative
    boundBone {boneName=dyn_dogtags_02}
    componentAppliedLinearAccClamp=<3000 3000 3000>
    componentLinearAccScale=<0.5 0.5 0.5>
    componentLinearVelScale=<0.5 0.5 0.5>
    gravityOverride {z=-980}
    physicsBodyDefinitions [{
      boundBone {boneName=dyn_dogtags_02}
      boxExtents=<30 30 30>
      constraintSetup {angularLimitsMax=<0 0 0>, angularLimitsMin=<0 -10 -20>, angularTarget=<1 0 0>}
      localJointOffset=<8 0 0>
     }]
    simSpaceSettings {maxAngularAcceleration=35, maxAngularVelocity=7, simSpaceAngularAlpha=0.5}
   }
   n10 = Anim Dynamics - Bone: dyn_zipper!(ComponentPose=@n9, Alpha=1, ComponentLinearAccScale:Vector="0.5,0.5,0.5", ComponentLinearVelScale:Vector="0.5,0.5,0.5") -> (Pose) {
    angularDampingOverride=1
    angularSpringConstant=20000
    bAngularSpring=true
    bGravityOverrideInSimSpace=true
    bOverrideAngularDamping=true
    bUseGravityOverride=true
    gravityScale=0.35
    simulationSpace=RootRelative
    boundBone {boneName=dyn_zipper}
    componentAppliedLinearAccClamp=<3000 3000 3000>
    componentLinearAccScale=<0.5 0.5 0.5>
    componentLinearVelScale=<0.5 0.5 0.5>
    gravityOverride {z=-980}
    physicsBodyDefinitions [{
      boundBone {boneName=dyn_zipper}
      boxExtents=<20 20 20>
      constraintSetup {angularLimitsMax=<0 0 70>, angularLimitsMin=<0 0 -70>, angularTarget=<1 0 0>}
      localJointOffset=<8 0 0>
     }]
    simSpaceSettings {maxAngularAcceleration=35, maxAngularVelocity=7, simSpaceAngularAlpha=1}
   }
   Save cached pose 'dynamicsCache'(Pose=Component To Local(ComponentPose=@n10)) {cacheName=dynamicsCache}
  ]}
 EventGraph {nodes [
   Event BlueprintUpdateAnimation()
   > n1 = Sequence()
   n2 = "Gravity Override State Function (Parameter Asset) (Nativized)"(
    execute=@n1.then_1
    CharacterPartAnimInstance=Self-Reference()
    GameTimeInSeconds=GetGameTimeInSeconds()
    Parameters=&/M_MED_FunkyBear_Parameters
   )
   > Set GravityOverride(GravityOverride=@n2)
   > Set crouching alpha(crouching alpha:double=float - float(A=1, B=GetCurveValue(CurveName=crouching alpha))) -> (Output_Get:double)
   n3 = GetOwnerSkeletalMeshComponent()
   n4 = BreakRotator(InRot=BreakTransform(InTransform=GetDeltaTransformFromRefPose(self=@n3, BoneName=head, BaseName=neck_02)).Rotation)
   n5 = Get Y percentage() -> (Y percentage:float)
   n6 = Get X percentage() -> (X percentage:float)
   n7 = Get Z percentage() -> (Z percentage:float)
   > Set head rot(head rot:Rotator=MakeRotator(
     Roll=float * float!(A=@n4, B=-1, C:double=@n6) -> (ReturnValue)
     Pitch=float * float(A=@n4.Pitch, B=-1, C=@n5)
     Yaw=float * float(A=@n4.Yaw, B=-1, C=@n7)
    )) -> (Output_Get:Rotator)
   > n8 = Branch(Condition=Get bIsSkyDiving())
   > Set Is SkyDive Jacket(Is SkyDive Jacket:double=0.5) -> (Output_Get:double)
   Set Is SkyDive Jacket(execute=@n8.else, Is SkyDive Jacket=1)
   Branch(execute=@n1.then_0, Condition=Get bIsInFrontEnd())
   > Set Custom Idle Is Playing Float(Custom Idle Is Playing Float:double=GetCurveValue(self=GetOwnerFrontEndAnimInstance(), CurveName=Frontend_Idle_Bypass)) -> (Output_Get:double)
   n9 = Event BlueprintProcessAnimInputEvent()
   n10 = IsNamed(self=@n9.AnimInputEvent, InEventName=aie.Player.Zipline.Entered)
   n11 = OR Boolean(A=Get bIsMoving2D(), B=Get bIsTacticalSprinting())
   n12 = Get bIsJumping()
   > n13 = Branch(Condition=OR Boolean(A=@n10, B=AND Boolean!(A=@n10, B=@n12, C:bool=@n11) -> (ReturnValue)))
   > Set animDyn on zipline(animDyn on zipline:Vector="0, 0, 0") -> (Output_Get:Vector)
   Branch(execute=@n13.else, Condition=IsNamed(self=@n9.AnimInputEvent, InEventName=aie.Player.Zipline.Exited))
   > Set animDyn on zipline(animDyn on zipline="0.5,0.5,0.5")
  ]}
]
variables [
 animDyn on zipline {type=Vector, default=<0.5 0.5 0.5>}
 crouching alpha {type=number}
 Custom Idle Is Playing Float {type=number}
 GravityOverride {type=Vector}
 head rot {type=Rotator}
 Is SkyDive Jacket {type=number, default=1}
 X percentage {type=number, default=0.8}
 Y percentage {type=number, default=0.8}
 Z percentage {type=number, default=0.4}
]
