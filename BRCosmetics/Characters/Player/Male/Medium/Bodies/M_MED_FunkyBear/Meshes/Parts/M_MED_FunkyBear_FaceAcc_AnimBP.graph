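# Post-process-style AnimBP for the FunkyBear face accessory, adding secondary physics motion to its ears and stuffing.
# It copies the attached parent's pose, passes it through an additive head-bone modifier, then chains spring-based, root-relative AnimDynamics on each dynamic bone.
# The update event computes a scaled inverse of head rotation relative to neck_02 into `head rot`.
# Zipline enter/exit input events lower and restore the `animDyn on zipline` scale vector.
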AnimBlueprint /BRCosmetics/Characters/Player/Male/Medium/Bodies/M_MED_FunkyBear/Meshes/Parts/M_MED_FunkyBear_FaceAcc_AnimBP
target_skeleton=&/M_MED_FunkyBear_FaceAcc_Skeleton
graphs [
 AnimGraph {nodes [
   Output Pose(Result=Use cached pose 'dynamicsCache'()) {layerGroup=None}
   n1 = Anim Dynamics - Bone: dyn_stuffing_04_cr!(
    ComponentPose=Local To Component(LocalPose=Use cached pose 'modifyTransPreCache'())
    Alpha=1
    ComponentLinearAccScale:Vector="0.5,0.5,0.5"
    ComponentLinearVelScale:Vector="0.5,0.5,0.5"
   ) -> (Pose) {
    angularDampingOverride=1
    angularSpringConstant=20000
    bAngularSpring=true
    bGravityOverrideInSimSpace=true
    bOverrideAngularDamping=true
    bUseGravityOverride=true
    gravityScale=0.35
    simulationSpace=RootRelative
    boundBone {boneName=dyn_stuffing_04_cr}
    componentAppliedLinearAccClamp=<3000 3000 3000>
    componentLinearAccScale=<0.5 0.5 0.5>
    componentLinearVelScale=<0.5 0.5 0.5>
    gravityOverride {z=-980}
    physicsBodyDefinitions [{
      boundBone {boneName=dyn_stuffing_04_cr}
      boxExtents=<35 35 35>
      constraintSetup {angularLimitsMax=<0 20 20>, angularLimitsMin=<0 -20 -20>, angularTarget=<1 0 0>}
      localJointOffset=<5 0 0>
     }]
    simSpaceSettings {maxAngularAcceleration=35, maxAngularVelocity=7, simSpaceAngularAlpha=0.5}
   }
   n2 = Anim Dynamics - Bone: dyn_stuffing_02_cr!(ComponentPose=@n1, Alpha=1, ComponentLinearAccScale:Vector="0.5,0.5,0.5", ComponentLinearVelScale:Vector="0.5,0.5,0.5") -> (Pose) {
    angularDampingOverride=1
    angularSpringConstant=20000
    bAngularSpring=true
    bGravityOverrideInSimSpace=true
    bOverrideAngularDamping=true
    bUseGravityOverride=true
    gravityScale=0.35
    simulationSpace=RootRelative
    boundBone {boneName=dyn_stuffing_02_cr}
    componentAppliedLinearAccClamp=<3000 3000 3000>
    componentLinearAccScale=<0.5 0.5 0.5>
    componentLinearVelScale=<0.5 0.5 0.5>
    gravityOverride {z=-980}
    physicsBodyDefinitions [{
      boundBone {boneName=dyn_stuffing_02_cr}
      boxExtents=<35 35 35>
      constraintSetup {angularLimitsMax=<0 20 20>, angularLimitsMin=<0 -20 -20>, angularTarget=<1 0 0>}
      localJointOffset=<5 0 0>
     }]
    simSpaceSettings {maxAngularAcceleration=35, maxAngularVelocity=7, simSpaceAngularAlpha=0.5}
   }
   n3 = Anim Dynamics - Bone: dyn_stuffing_01_cr!(ComponentPose=@n2, Alpha=1, ComponentLinearAccScale:Vector="0.5,0.5,0.5", ComponentLinearVelScale:Vector="0.5,0.5,0.5") -> (Pose) {
    angularDampingOverride=1
    angularSpringConstant=20000
    bAngularSpring=true
    bGravityOverrideInSimSpace=true
    bOverrideAngularDamping=true
    bUseGravityOverride=true
    gravityScale=0.35
    simulationSpace=RootRelative
    boundBone {boneName=dyn_stuffing_01_cr}
    componentAppliedLinearAccClamp=<3000 3000 3000>
    componentLinearAccScale=<0.5 0.5 0.5>
    componentLinearVelScale=<0.5 0.5 0.5>
    gravityOverride {z=-980}
    physicsBodyDefinitions [{
      boundBone {boneName=dyn_stuffing_01_cr}
      boxExtents=<35 35 35>
      constraintSetup {angularLimitsMax=<0 20 20>, angularLimitsMin=<0 -20 -20>, angularTarget=<1 0 0>}
      localJointOffset=<5 0 0>
     }]
    simSpaceSettings {maxAngularAcceleration=35, maxAngularVelocity=7, simSpaceAngularAlpha=0.5}
   }
   n4 = Anim Dynamics - Bone: dyn_stuffing_03_cr!(ComponentPose=@n3, Alpha=1, ComponentLinearAccScale:Vector="0.5,0.5,0.5", ComponentLinearVelScale:Vector="0.5,0.5,0.5") -> (Pose) {
    angularDampingOverride=1
    angularSpringConstant=20000
    bAngularSpring=true
    bGravityOverrideInSimSpace=true
    bOverrideAngularDamping=true
    bUseGravityOverride=true
    gravityScale=0.35
    simulationSpace=RootRelative
    boundBone {boneName=dyn_stuffing_03_cr}
    componentAppliedLinearAccClamp=<3000 3000 3000>
    componentLinearAccScale=<0.5 0.5 0.5>
    componentLinearVelScale=<0.5 0.5 0.5>
    gravityOverride {z=-980}
    physicsBodyDefinitions [{
      boundBone {boneName=dyn_stuffing_03_cr}
      boxExtents=<35 35 35>
      constraintSetup {angularLimitsMax=<0 20 20>, angularLimitsMin=<0 -20 -20>, angularTarget=<1 0 0>}
      localJointOffset=<5 0 0>
     }]
    simSpaceSettings {maxAngularAcceleration=35, maxAngularVelocity=7, simSpaceAngularAlpha=0.5}
   }
   n5 = Anim Dynamics - Bone: dyn_ears_01_rt!(ComponentPose=@n4, Alpha=1, ComponentLinearAccScale:Vector="0.5,0.5,0.5", ComponentLinearVelScale:Vector="0.5,0.5,0.5") -> (Pose) {
    angularDampingOverride=1
    angularSpringConstant=20000
    bAngularSpring=true
    bGravityOverrideInSimSpace=true
    bOverrideAngularDamping=true
    bUseGravityOverride=true
    gravityScale=0.35
    simulationSpace=RootRelative
    boundBone {boneName=dyn_ears_01_rt}
    componentAppliedLinearAccClamp=<3000 3000 3000>
    componentLinearAccScale=<0.5 0.5 0.5>
    componentLinearVelScale=<0.5 0.5 0.5>
    gravityOverride {z=-980}
    physicsBodyDefinitions [{
      boundBone {boneName=dyn_ears_01_rt}
      boxExtents=<40.25 40.25 40.25>
      constraintSetup {angularLimitsMax=<0 30 20>, angularLimitsMin=<0 -30 -8>, angularTarget=<1 0 0>}
      localJointOffset=<5.1 0 0>
     }]
    simSpaceSettings {maxAngularAcceleration=35, maxAngularVelocity=7, simSpaceAngularAlpha=0.5}
   }
   n6 = Anim Dynamics - Bone: dyn_ears_02_rt!(ComponentPose=@n5, Alpha=1, ComponentLinearAccScale:Vector="0.5,0.5,0.5", ComponentLinearVelScale:Vector="0.5,0.5,0.5") -> (Pose) {
    angularDampingOverride=1
    angularSpringConstant=20000
    bAngularSpring=true
    bGravityOverrideInSimSpace=true
    bOverrideAngularDamping=true
    bUseGravityOverride=true
    gravityScale=0.35
    simulationSpace=RootRelative
    boundBone {boneName=dyn_ears_02_rt}
    componentAppliedLinearAccClamp=<3000 3000 3000>
    componentLinearAccScale=<0.5 0.5 0.5>
    componentLinearVelScale=<0.5 0.5 0.5>
    gravityOverride {z=-980}
    physicsBodyDefinitions [{
      boundBone {boneName=dyn_ears_02_rt}
      boxExtents=<40.25 40.25 40.25>
      constraintSetup {angularLimitsMax=<0 10 10>, angularLimitsMin=<0 -10 -10>, angularTarget=<1 0 0>}
      localJointOffset=<5.1 0 0>
     }]
    simSpaceSettings {maxAngularAcceleration=35, maxAngularVelocity=7, simSpaceAngularAlpha=0.5}
   }
   n7 = Anim Dynamics - Bone: dyn_ears_03_rt!(ComponentPose=@n6, Alpha=1, ComponentLinearAccScale:Vector="0.5,0.5,0.5", ComponentLinearVelScale:Vector="0.5,0.5,0.5") -> (Pose) {
    angularDampingOverride=1
    angularSpringConstant=20000
    bAngularSpring=true
    bGravityOverrideInSimSpace=true
    bOverrideAngularDamping=true
    bUseGravityOverride=true
    gravityScale=0.35
    simulationSpace=RootRelative
    boundBone {boneName=dyn_ears_03_rt}
    componentAppliedLinearAccClamp=<3000 3000 3000>
    componentLinearAccScale=<0.5 0.5 0.5>
    componentLinearVelScale=<0.5 0.5 0.5>
    gravityOverride {z=-980}
    physicsBodyDefinitions [{
      boundBone {boneName=dyn_ears_03_rt}
      boxExtents=<40.25 40.25 40.25>
      constraintSetup {angularLimitsMax=<0 10 10>, angularLimitsMin=<0 -10 -10>, angularTarget=<1 0 0>}
      localJointOffset=<5.1 0 0>
     }]
    simSpaceSettings {maxAngularAcceleration=35, maxAngularVelocity=7, simSpaceAngularAlpha=0.5}
   }
   n8 = Anim Dynamics - Bone: dyn_ears_01_lf!(ComponentPose=@n7, Alpha=1, ComponentLinearAccScale:Vector="0.5,0.5,0.5", ComponentLinearVelScale:Vector="0.5,0.5,0.5") -> (Pose) {
    angularDampingOverride=1
    angularSpringConstant=17000
    bAngularSpring=true
    bGravityOverrideInSimSpace=true
    bOverrideAngularDamping=true
    bUseGravityOverride=true
    gravityScale=0.35
    simulationSpace=RootRelative
    boundBone {boneName=dyn_ears_01_lf}
    componentAppliedLinearAccClamp=<3000 3000 3000>
    componentLinearAccScale=<0.5 0.5 0.5>
    componentLinearVelScale=<0.5 0.5 0.5>
    gravityOverride {z=-980}
    physicsBodyDefinitions [{
      boundBone {boneName=dyn_ears_01_lf}
      boxExtents=<36.8 36.8 36.8>
      constraintSetup {angularLimitsMax=<0 20 20>, angularLimitsMin=<0 -20 -20>, angularTarget=<1 0 0>}
      localJointOffset=<5.1 0 0>
     }]
    simSpaceSettings {maxAngularAcceleration=35, maxAngularVelocity=7, simSpaceAngularAlpha=0.5}
   }
   n9 = Anim Dynamics - Bone: dyn_ears_02_lf!(ComponentPose=@n8, Alpha=1, ComponentLinearAccScale:Vector="0.5,0.5,0.5", ComponentLinearVelScale:Vector="0.5,0.5,0.5") -> (Pose) {
    angularDampingOverride=1
    angularSpringConstant=17000
    bAngularSpring=true
    bGravityOverrideInSimSpace=true
    bOverrideAngularDamping=true
    bUseGravityOverride=true
    gravityScale=0.35
    simulationSpace=RootRelative
    boundBone {boneName=dyn_ears_02_lf}
    componentAppliedLinearAccClamp=<3000 3000 3000>
    componentLinearAccScale=<0.5 0.5 0.5>
    componentLinearVelScale=<0.5 0.5 0.5>
    gravityOverride {z=-980}
    physicsBodyDefinitions [{
      boundBone {boneName=dyn_ears_02_lf}
      boxExtents=<36.8 36.8 36.8>
      constraintSetup {angularLimitsMax=<0 20 20>, angularLimitsMin=<0 -20 -20>, angularTarget=<1 0 0>}
      localJointOffset=<5.1 0 0>
     }]
    simSpaceSettings {maxAngularAcceleration=35, maxAngularVelocity=7, simSpaceAngularAlpha=0.5}
   }
   n10 = Anim Dynamics - Bone: dyn_ears_03_lf!(ComponentPose=@n9, Alpha=1, ComponentLinearAccScale:Vector="0.5,0.5,0.5", ComponentLinearVelScale:Vector="0.5,0.5,0.5") -> (Pose) {
    angularDampingOverride=1
    angularSpringConstant=17000
    bAngularSpring=true
    bGravityOverrideInSimSpace=true
    bOverrideAngularDamping=true
    bUseGravityOverride=true
    gravityScale=0.35
    simulationSpace=RootRelative
    boundBone {boneName=dyn_ears_03_lf}
    componentAppliedLinearAccClamp=<3000 3000 3000>
    componentLinearAccScale=<0.5 0.5 0.5>
    componentLinearVelScale=<0.5 0.5 0.5>
    gravityOverride {z=-980}
    physicsBodyDefinitions [{
      boundBone {boneName=dyn_ears_03_lf}
      boxExtents=<36.8 36.8 36.8>
      constraintSetup {angularLimitsMax=<0 10 10>, angularLimitsMin=<0 -10 -10>, angularTarget=<1 0 0>}
      localJointOffset=<5.1 0 0>
     }]
    simSpaceSettings {maxAngularAcceleration=35, maxAngularVelocity=7, simSpaceAngularAlpha=0.5}
   }
   Save cached pose 'dynamicsCache'(Pose=Component To Local(ComponentPose=@n10)) {cacheName=dynamicsCache}
   n11 = "Transform (Modify) Bone - Bone: head"!(ComponentPose=Local To Component(LocalPose=Copy Pose From Mesh() {bUseAttachedParent=true}), Alpha=1, Rotation="0,0,0") -> (Pose) {rotationMode=BMM_Additive, rotationSpace=BCS_BoneSpace, alphaBoolBlend {blendOption=Linear}, boneToModify {boneName=head}}
   Save cached pose 'modifyTransPreCache'(Pose=Component To Local(ComponentPose=@n11)) {cacheName=modifyTransPreCache}
  ]}
 EventGraph {nodes [
   Event BlueprintUpdateAnimation()
   n1 = GetOwnerSkeletalMeshComponent()
   n2 = BreakRotator(InRot=BreakTransform(InTransform=GetDeltaTransformFromRefPose(self=@n1, BoneName=head, BaseName=neck_02)).Rotation)
   n3 = Get Y percentage() -> (Y percentage:float)
   n4 = Get X percentage() -> (X percentage:float)
   n5 = Get Z percentage() -> (Z percentage:float)
   > Set head rot(head rot:Rotator=MakeRotator(
     Roll=float * float!(A=@n2, B=-1, C:double=@n4) -> (ReturnValue)
     Pitch=float * float(A=@n2.Pitch, B=-1, C=@n3)
     Yaw=float * float(A=@n2.Yaw, B=-1, C=@n5)
    )) -> (Output_Get:Rotator)
   n6 = Event BlueprintProcessAnimInputEvent()
   n7 = IsNamed(self=@n6.AnimInputEvent, InEventName=aie.Player.Zipline.Entered)
   n8 = OR Boolean(A=Get bIsMoving2D(), B=Get bIsTacticalSprinting())
   n9 = Get bIsJumping()
   > n10 = Branch(Condition=OR Boolean(A=@n7, B=AND Boolean!(A=@n7, B=@n9, C:bool=@n8) -> (ReturnValue)))
   > Set animDyn on zipline(animDyn on zipline:Vector="0.1,0.1,0.1") -> (Output_Get:Vector)
   Branch(execute=@n10.else, Condition=IsNamed(self=@n6.AnimInputEvent, InEventName=aie.Player.Zipline.Exited))
   > Set animDyn on zipline(animDyn on zipline="0.5,0.5,0.5")
  ]}
]
variables [
 animDyn on zipline {type=Vector, default=<0.5 0.5 0.5>}
 head rot {type=Rotator}
 X percentage {type=number, default=0.8}
 Y percentage {type=number, default=0.8}
 Z percentage {type=number, default=0.4}
]
